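; Spire variant low-altitude cloud Blueprint that fades a cloud mesh and manages its ambient and lightning audio.
; Timelines drive custom primitive data for fade in and out, while other events write size limits and an entrance-glow start time.
; A cosmetic-only function watches MetaSound lightning outputs, storing amplitude; the trigger handler is empty.
; BeginPlay starts watching and fades in; fade-out completion stops both loops and unwatches.
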(event EventBeginPlay
  (|CallParentFunction)
  (CallFunction|InitializeLightningAudioWatcher true)
  (CallFunction|FadeinCloud))

(event Custom|SetCustomParameters
  (params (MeshSeed :type Float%20%28double-precision%29) (MaxSize :type Float%20%28double-precision%29) (MinSize :type Float%20%28double-precision%29))
  (bind _cloudmesh (Variables|Default|GetCloudMesh))
  (Rendering|Material|SetDefaultCustomPrimitiveDataFloat _cloudmesh 1 MaxSize)
  (Rendering|Material|SetDefaultCustomPrimitiveDataFloat _cloudmesh 2 MinSize))

(event Custom|FadeOutCloud
  (bind (_direction _newtrack_0) (Timeline|CloudPhaseValue Play)
    (:Update
      (Rendering|Material|SetCustomPrimitiveDataFloat (Variables|Default|GetCloudMesh) 12 _newtrack_0))
    (:Finished
      (Audio|Components|Audio|FadeOut (Variables|Default|GetAmbient_Cloud_Loop) 1.0 0.0 "Linear" :!class /Script/Engine.AudioComponent)
      (CallFunction|InitializeLightningAudioWatcher false))))

(event Custom|FadeInCloud
  (bind (_direction _newtrack_0) (Timeline|Timeline ReverseFromEnd)
    (:Update
      (Rendering|Material|SetCustomPrimitiveDataFloat (Variables|Default|GetCloudMesh) 12 _newtrack_0))))

(event Custom|SpireEntranceGlow
  (Rendering|Material|SetCustomPrimitiveDataFloat (Variables|Default|GetCloudMesh) 9 (Utilities|Time|GetGameTimeinSeconds)))

(event Custom|OnLightningTriggered
  (params (OutputName :type Name) (Output :type Meta%20Sound%20Output%20Structure%20%28by%20ref%29)))

(event Custom|OnLightningAmplitudeChanged
  (params (OutputName :type Name) (Output :type Meta%20Sound%20Output%20Structure%20%28by%20ref%29))
  ;[0 0 0 0.5] Audio reactive events for lightning audio
  (bind (_success _returnvalue) (MetaSoundOutput|GetFloat Output))
  (if _success
    (Variables|Default|SetLightningAudioAmplitude _returnvalue)))

(fn InitializeLightningAudioWatcher
  (params (bInitialize :type Boolean))
  (bind _returnvalue (WorldSubsystems|GetMetaSoundOutputSubsystem))
  (Networking|CanExecuteCosmeticEvents
    (:True
      (Variables|Default|GetLightning_Loop
        (:then
          (if bInitialize
            (MetaSoundOutput|WatchOutput _returnvalue (Variables|Default|GetLightning_Loop) (Variables|Default|GetTriggerParamName) (EventDispatchers|CreateEvent :!event OnLightningTriggered) "None" "None" :!class /Script/MetasoundEngine.MetaSoundOutputSubsystem)
            (MetaSoundOutput|WatchOutput _returnvalue (Variables|Default|GetLightning_Loop) (Variables|Default|GetAmplitudeParamName) (EventDispatchers|CreateEvent :!event OnLightningAmplitudeChanged) "None" "None" :!class /Script/MetasoundEngine.MetaSoundOutputSubsystem)
            (else
              (Audio|Components|Audio|FadeOut (Variables|Default|GetLightning_Loop) 1.0 0.0 "Linear" :!class /Script/Engine.AudioComponent)
              (MetaSoundOutput|UnwatchOutput _returnvalue (Variables|Default|GetLightning_Loop) (Variables|Default|GetTriggerParamName) (EventDispatchers|CreateEvent :!event OnLightningTriggered) "None" "None")
              (MetaSoundOutput|UnwatchOutput _returnvalue (Variables|Default|GetLightning_Loop) (Variables|Default|GetAmplitudeParamName) (EventDispatchers|CreateEvent :!event OnLightningAmplitudeChanged) "None" "None"))))
        (:else)))))

(fn ConstructionScript
  (|CallParentFunction))
