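# Material function for the Spire clouds that outputs lightning colors, a lightning mask and an inner glow color.
# Three time-driven triangle pulses, computed per vertex, select lightning texture channels, shaped by the Masks input and combined with sphere-masked glow flicker.
# Actor position drives hashed time offsets and shifts for object-stabilized screen-space UVs, varying each instance.
# Custom primitive data holds the start time of the animated entrance glow sphere.
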MaterialFunction /SpireSharedAssets/Assets/Materials/MF_SpireCloudLightning
graph {nodes [
  ComponentMask() {b=true}
  ComponentMask(None=TextureCoordinate()) {r=true}
  n1 = Multiply(A=NamedRerouteUsage(), B=ScalarParameter() {defaultValue=0.5, parameterName=Light_Tiling})
  n2 = TextureSample(UVs=@n1) {
   samplerSource=SSM_Wrap_WorldGroupSettings
   samplerType=SAMPLERTYPE_LinearColor
   texture=/SpireSharedAssets/Assets/Textures/T_SpireClouds_Lightning
  }
  n3 = TextureSample(UVs=@n1) {
   samplerSource=SSM_Wrap_WorldGroupSettings
   samplerType=SAMPLERTYPE_LinearColor
   texture=&T_SpireClouds_Lightning
  }
  n4 = LinearInterpolate(
   A=ScalarParameter() {defaultValue=-1, parameterName=Min_TimeOffset}
   B=ScalarParameter() {defaultValue=1, parameterName=Max_TimeOffset}
   Alpha=NamedRerouteUsage()
  )
  n5 = Multiply(
   A=Add(A=Time(), B=NamedRerouteDeclaration(None=@n4) {name=TimeOffsetValues, nodeColor=<rgba 1 0 0.1875 1>})
   B=Constant3Vector() {constant=<rgba 0.424 0.87 0.168 0>}
  )
  n6 = Abs(None=Add(A=Frac(None=@n5)) {constB=-0.5})
  n7 = Abs(None=Add(A=Multiply(A=@n6) {constB=2}) {constB=-0.2})
  n8 = Add(A=OneMinus(None=Multiply(A=@n7) {constB=1.25})) {constB=-0.6}
  n9 = DotProduct(A=@n3.RGB, B=VertexInterpolator(VS=Multiply(A=@n8) {constB=2}).PS)
  n10 = NamedRerouteDeclaration(None=FunctionInput(Preview=Constant()) {bUsePreviewValueAsDefault=true, inputName=Masks_Input, inputType=FunctionInput_Scalar}) {name=Masks, nodeColor=<rgba 0 1 0.4141 1>}
  n11 = ScalarParameter() {defaultValue=12, parameterName=MaskPower}
  n12 = Multiply(A=Multiply(A=@n2.R, B=Saturate(None=@n9)), B=Saturate(None=Power(Base=@n10, Exp=@n11) {constExponent=12}))
  n13 = Add(A=NamedRerouteUsage(), B=Multiply(A=Multiply(A=@n12, B=@n12)) {constB=5})
  n14 = VectorParameter() {group=Lightning/Thunder Parameters, parameterName=LightningColor, defaultValue=<rgba 1 0 0.1217 1>}
  FunctionOutput(None=Multiply(A=Multiply(A=@n13, B=@n14.RGB), B=NamedRerouteUsage())) {outputName=LightningColors}
  FunctionOutput(None=NamedRerouteUsage()) {outputName=Lightning_Mask, sortPriority=1}
  n15 = VectorParameter() {parameterName=InnerGlow_Color, defaultValue=<rgba 1 0 0.3179 1>}
  n16 = WorldPosition()
  n17 = Add(A=ActorPositionWS(), B=VectorParameter() {parameterName=Glow_Offset, defaultValue=<rgba 0 0 -4000 1>}.RGB)
  n18 = Time()
  n19 = ScalarParameter() {bUseCustomPrimitiveData=true, defaultValue=-10000000000, parameterName=C_EntranceGlow, primitiveDataIndex=9}
  n20 = Saturate(None=Divide(A=Subtract(A=@n18, B=@n19), B=ScalarParameter() {defaultValue=8, parameterName=EntranceGlowDuration}))
  n21 = ConstantBiasScale(None=Sine(None=Add(A=@n20) {constB=-0.25}))
  n22 = ScalarParameter() {defaultValue=2, parameterName=AnimCurvemult}
  n23 = Multiply(A=ScalarParameter() {defaultValue=5000, parameterName=MaskRadius}, B=Saturate(None=Multiply(A=@n21, B=@n22) {constB=4}))
  n24 = ScalarParameter() {parameterName=MaskHardness}
  n25 = NamedRerouteUsage()
  n26 = Power(
   Base=Multiply(A=OneMinus(None=@n25), B=ScalarParameter() {defaultValue=1, parameterName=MaskMult})
   Exp=ScalarParameter() {defaultValue=1, parameterName=MaskPow}
  )
  n27 = ScalarParameter() {defaultValue=2, parameterName=Masks_Strength}
  n28 = Saturate(None=Subtract(A=SphereMask(A=@n16.XYZ, B=@n17, Radius=@n23, Hardness=@n24) {hardnessPercent=0}, B=Multiply(A=@n26, B=@n27)))
  FunctionOutput(None=Multiply(A=Multiply(A=@n15.RGB, B=@n28), B=ScalarParameter() {defaultValue=1, parameterName=InnerGlow_Str})) {outputName=InnerGlowColors, sortPriority=2}
  n29 = Multiply(A=OneMinus(None=NamedRerouteUsage()), B=ScalarParameter() {defaultValue=1, parameterName=LightningGlowMaskMult})
  Multiply(A=@n29) {constB=2}
  n30 = WorldPosition()
  n31 = Add(A=ActorPositionWS(), B=VectorParameter() {parameterName=LightningGlow_Offset, defaultValue=<rgba 0 0 0 1>}.RGB)
  n32 = ScalarParameter() {defaultValue=15000, parameterName=LightningGlowRadius}
  n33 = ScalarParameter() {defaultValue=0.25, parameterName=LightningMaskHardness}
  n34 = Multiply(
   A=Subtract(A=SphereMask(A=@n30.XYZ, B=@n31, Radius=@n32, Hardness=@n33) {hardnessPercent=0}, B=@n29)
   B=LinearInterpolate(
    A=ScalarParameter() {defaultValue=0.25, parameterName=NoiseGlow_MinInfluence}
    B=ScalarParameter() {defaultValue=1, parameterName=NoiseGlow_MaxInfluence}
    Alpha=@n2.R
   ) {constA=0.25}
  )
  n35 = Multiply(A=Add(A=Time(), B=NamedRerouteUsage()), B=ScalarParameter() {defaultValue=0.1, parameterName=GlowSpeed}) {constB=0.15}
  n36 = TextureSample(UVs=AppendVector(A=Frac(None=@n35), B=Constant())) {
   samplerSource=SSM_Wrap_WorldGroupSettings
   samplerType=SAMPLERTYPE_LinearColor
   texture=&T_SpireClouds_Lightning
  }
  n37 = LinearInterpolate(
   A=ScalarParameter() {defaultValue=0.2, parameterName=minLightningGlow}
   B=ScalarParameter() {defaultValue=0.8, parameterName=MaxLightningGlow}
   Alpha=@n36.R
  ) {constA=0.2}
  NamedRerouteDeclaration(None=Saturate(None=Multiply(A=@n34, B=@n37))) {name=LightningGlow, nodeColor=<rgba 1 0 0.1641 1>}
  n38 = ActorPositionWS()
  n39 = NamedRerouteUsage()
  n40 = Divide(A=@n38, B=VectorParameter() {parameterName=Lightning_ThunderPositionOffsetBias, defaultValue=<rgba 8000 8000 8000 1>}.RGB)
  n41 = MaterialFunctionCall(Float3=ConstantBiasScale(None=Sine(None=@n40))) {materialFunction=&BreakOutFloat3Components}
  n42 = MaterialFunctionCall(X=Add(A=@n41.R, B=@n41.R), Y=Add(A=@n41.G, B=@n41.G)) {materialFunction=&MakeFloat2}
  NamedRerouteDeclaration(None=Add(
    A=MaterialFunctionCall!(Position=@n38, UVRadialScale=@n39) {materialFunction=/Game/Weapons/WeaponSkins/MaterialFunction/ObjectStabilizedScreenSpaceUVs}.StabilizedUVs
    B=@n42.Result
   )) {name=ScreenUVCoords, nodeColor=<rgba 0 0.7422 1 1>}
  NamedRerouteDeclaration(None=Saturate(None=@n13)) {name=LightningMask, nodeColor=<rgba 0 1 0.8594 1>}
  n43 = MaterialFunctionCall(Float3=@n40) {materialFunction=&BreakOutFloat3Components}
  n44 = Add(A=Add(A=@n43.R, B=@n43.R), B=@n43.B)
  NamedRerouteDeclaration(None=Abs(None=MaterialFunctionCall(Float=@n44) {materialFunction=&MF_Hash}.Float)) {name=PositionalTimeOffset, nodeColor=<rgba 0.6172 0 1 1>}
  n45 = SphereMask(
   A=WorldPosition().XYZ
   B=Add(A=ActorPositionWS(), B=VectorParameter() {parameterName=CloudBounds_Center, defaultValue=<rgba 0 0 -4000 1>}.RGB)
   Radius=ScalarParameter() {defaultValue=9000, parameterName=CloudBoundsRadius}
   Hardness=ScalarParameter() {parameterName=CloudBoundsHardness}
  ) {hardnessPercent=0}
  n46 = Multiply(A=@n45, B=@n45)
  n47 = ScalarParameter() {defaultValue=1.5, parameterName=CloudBoundsMult}
  NamedRerouteDeclaration(None=Saturate(None=Multiply(A=@n46, B=@n47))) {name=CloudBoundsMask, nodeColor=<rgba 0.3594 0 1 1>}
  NamedRerouteDeclaration(None=@n42.Result) {name=SCreenUVOffsets, nodeColor=<rgba 1 0.9609 0 1>}
  ScalarParameter() {defaultValue=1, parameterName=LightningGlowMaskMult}
  VectorParameter() {parameterName=Lighting Color, defaultValue=<rgba 1 0.1233 0.1233 1>}
  n48 = ComponentMask(None=WorldPosition().XYZ) {g=true, r=true}
  n49 = NamedRerouteDeclaration(None=ScalarParameter() {defaultValue=1, parameterName=Lightning_Tiling}) {name=LightningTiling, nodeColor=<rgba 0 0.4141 1 1>}
  n50 = Time()
  n51 = Constant2Vector() {g=0.004, r=0.01}
  VertexInterpolator(VS=Add(A=Divide(A=@n48, B=@n49) {constB=2000}, B=Multiply(A=@n50, B=@n51)))
 ]}
