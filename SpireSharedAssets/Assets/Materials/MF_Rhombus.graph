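# Material function that produces a rhombus signed-distance shape with animated edge distortion.
# Input UVs are centred, then offset by the RG of a tiled, panning distortion normal map scaled by DistortionIntensity.
# A Custom HLSL node computes the distance for RhombusSize, returning Mask (positive inside), its negation as InvertedMask, and UVs from the size-normalised offset to the nearest edge.
# Outputs are raw distances, with no threshold or smoothing applied.
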MaterialFunction /SpireSharedAssets/Assets/Materials/MF_Rhombus
graph {nodes [
  MaterialExpressionFunctionInput_1 = FunctionInput(Preview=ComponentMask(None=VectorParameter() {parameterName=RhombusSize, defaultValue=<rgba 0.31 0.4 0.5 0.5>}.RGB) {g=true, r=true}) {bUsePreviewValueAsDefault=true, inputName=RhombusSize, inputType=FunctionInput_Vector2, previewValue=<0.31 0.4 0 1>}
  n1 = FunctionInput(Preview=TextureCoordinate()) {bUsePreviewValueAsDefault=true, inputName=UVs, inputType=FunctionInput_Vector2}
  n2 = TextureSample(UVs=Panner(Coordinate=Multiply(A=@n1) {constB=4}) {speedX=0.2, speedY=0.8}) {samplerType=SAMPLERTYPE_Normal, texture=&T_Weap_Distort_N_02}
  n3 = FunctionInput() {bUsePreviewValueAsDefault=true, inputName=DistortionIntensity, inputType=FunctionInput_Scalar, previewValue=<0.04 0 0 1>}
  MaterialExpressionAdd_10 = Add(A=Subtract(A=@n1) {constB=0.5}, B=Multiply(A=ComponentMask(None=@n2.RGB) {g=true, r=true}, B=@n3))
  n4 = Custom!(b=@MaterialExpressionFunctionInput_1, p=@MaterialExpressionAdd_10) {
   code="// Custom Rhombus Code\r\nfloat2 p_abs = abs(p);\r\nfloat nd = (b.x * (b.x - 2.0 * p_abs.x)) - (b.y * (b.y - 2.0 * p_abs.y));\r\nfloat h = clamp(nd / dot(b, b), -1.0, 1.0);\r\nfloat2 offset = 0.5 * b * float2(1.0 - h, 1.0 + h);\r\nfloat d = length(p_abs - offset);\r\nfloat sdf = -d * sign(p_abs.x * b.y + p_abs.y * b.x - b.x * b.y);\r\n// UVs\r\nrhombusUV = (p_abs - offset) / b;\r\n//Inverted\r\ninverted_mask = d * sign(p_abs.x * b.y + p_abs.y * b.x - b.x * b.y);\r\nreturn sdf;\r\n"
   desc="rhombus sdf\r\np= point we want to compute the distance\r\nb= shape size"
   outputType=CMOT_Float1
   showCode=true
   additionalOutputs [{outputName=rhombusUV, outputType=CMOT_Float2}, {outputName=inverted_mask, outputType=CMOT_Float1}]
   inputs [
    {inputName=b, input {expression=&self:MaterialExpressionFunctionInput_1}}
    {inputName=p, input {expression=&self:MaterialExpressionAdd_10}}
   ]
  }
  FunctionOutput(None=@n4.return) {outputName=Mask}
  FunctionOutput(None=@n4.rhombusUV) {outputName=UVs}
  FunctionOutput(None=@n4.inverted_mask) {outputName=InvertedMask}
 ]}
