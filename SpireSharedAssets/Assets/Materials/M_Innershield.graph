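# Unlit Modulate-blend material for Niagara mesh particles, apparently the inner shield layer of the Spire shield effect.
# Three panned samples of a chromatic caustics texture are averaged, tinted magenta-purple and lerped by Opacity.
# A vertical sine gradient on UV1 and a Fresnel-fed SchlickPhase term fade the result toward white, the neutral value for Modulate.
# Radial gradient and rhombus mask branches sit apart from that chain.
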Material /SpireSharedAssets/Assets/Materials/M_Innershield
graph {
 properties {
  bAutomaticallySetUsageInEditor=false
  bUsedWithNiagaraMeshParticles=true
  blendMode=BLEND_Modulate
  previewMesh=/Game/Effects/Fort_Effects/GameplayPulginsAssets/SPire/SM_FX_ShieldSpire3
  shadingModel=MSM_Unlit
  usedShadingModels=Unlit
 }
 nodes [
  n1 = MaterialFunctionCall!(
   UV=Multiply(A=TextureCoordinate(), B=ScalarParameter() {defaultValue=1, parameterName=UVScale})
   Rotation
   Speed=Constant() {r=0.1}
   Time
   Rotate UVs
   Fractional Part
  ) {materialFunction=&/MF_3UVPanner}
  n2 = Add(
   A=Add(
    A=TextureSample(UVs=@n1.UV1) {texture=/Game/Effects/Fort_Effects/Textures/Tile/T_ShieldCaustics_Tile_Chromatic}.RGB
    B=TextureSample(UVs=@n1.UV2) {texture=&T_ShieldCaustics_Tile_Chromatic}.RGB
   )
   B=TextureSample(UVs=@n1.UV3) {texture=&T_ShieldCaustics_Tile_Chromatic}.RGB
  )
  n3 = LinearInterpolate(
   A=Multiply(A=VectorParameter() {defaultValue=<rgba 6.0634 0.98 10 1>}.RGB, B=Divide(A=@n2) {constB=3})
   Alpha=ScalarParameter() {defaultValue=0.2, parameterName=Opacity}
  )
  n4 = Sine(None=ComponentMask(None=TextureCoordinate() {coordinateIndex=1, vTiling=0.5}) {g=true})
  n5 = ScalarParameter() {defaultValue=2.398, parameterName=ExponentIn}
  n6 = MaterialFunctionCall!(
   Cosine=OneMinus(None=Fresnel(ExponentIn=@n5) {baseReflectFraction=0})
   Anisotropy=ScalarParameter() {defaultValue=-0.8635, parameterName=Phase}
  ) {materialFunction=/Engine/Functions/Engine_MaterialFunctions01/Volumetrics/SchlickPhase}
  LinearInterpolate(A=LinearInterpolate(A=LinearInterpolate(B=@n3, Alpha=@n4) {constA=1}, Alpha=Saturate(None=@n6.Result)) {constA=1})
  MaterialFunctionCall(
   UVs=TextureCoordinate() {coordinateIndex=1}
   Radius=ScalarParameter() {defaultValue=0.1797, parameterName="Radius (S)"}
   Density=ScalarParameter() {defaultValue=1.531, parameterName="Density (S)"}
  ) {materialFunction=&RadialGradientExponential}
  Multiply(A=MaterialFunctionCall!(
    RhombusSize=VectorParameter() {parameterName=Param_1, defaultValue=<rgba -0.067 0.4 0 1>}.RGB
    UVs=TextureCoordinate() {coordinateIndex=1}
    DistortionIntensity
   ) {materialFunction=&/MF_Rhombus}.InvertedMask) {constB=4}
 ]
}
