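# Parent material for Spire Slag assets, layering an animated emissive glow and hit wobble onto the shared character material functions.
# Scrolling procedural bands and world-aligned noise, shaped by local height and radial falloff, tint with squared Emissive_Color.
# The FX texture masks this glow, using a mip-biased second sample as a soft halo, between the PreFX and PostFX functions.
# Destructible-damage animation adds World Position Offset, optionally scaled through a static switch.
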Material /SpireSharedAssets/Assets/Materials/M_SpireSlag_Parent
graph {
 properties {
  bUseMaterialAttributes=true
  bUsedWithClothing=true
  bUsedWithMeshDeformer=true
  bUsedWithMorphTargets=true
  bUsedWithNanite=true
  bUsedWithNiagaraMeshParticles=true
  bUsedWithSkeletalMesh=true
  pixelDepthOffsetMode=PDOM_Legacy
  shadingModel=MSM_FromMaterialExpression
  usedShadingModels=Default Lit | Subsurface | Clear Coat | Subsurface Profile | Cloth
 }
 nodes [
  n1 = LocalPosition()
  n2 = Saturate(None=Divide(A=Length(None=@n1.XY) {is=class MaterialExpressionLength}) {constB=250}) {desc=thickness}
  Multiply(A=Power(Base=OneMinus(None=@n2)) {constExponent=6}, B=Saturate(None=Divide(A=@n1.Z) {constB=250}) {desc=thickness})
  n3 = Constant()
  n4 = Divide(A=WorldPosition().XYZ, B=PreSkinnedLocalBounds().Max)
  n5 = Subtract(A=Multiply(A=ComponentMask(None=@n4) {b=true}) {constB=0.25}, B=Divide(A=Time()) {constB=2.8})
  n6 = MaterialFunctionCall!(Width=Constant() {r=0.3}, Sharpness=@n3, Offset=Constant() {r=-0.3}, Direction Switch, Compare, Input Coordinates=@n5) {materialFunction=/Engine/Functions/Engine_MaterialFunctions03/Procedurals/GeneratedBand}
  n7 = Multiply(A=MaterialFunctionCall(Width=Constant() {r=0.6}, Sharpness=@n3, Input Coordinates=@n5) {materialFunction=&GeneratedBand}.Result) {constB=0.1}
  n8 = TextureObject() {samplerType=SAMPLERTYPE_LinearGrayscale, texture=/Engine/EngineMaterials/Good64x64TilingNoiseHighFreq}
  n9 = Constant() {r=600}
  n10 = Add(A=Multiply(A=Time(), B=Constant3Vector() {constant=<rgba 0 0 -5 0>}), B=WorldPosition().XYZ)
  n11 = Add(A=Multiply(A=MaterialFunctionCall(TextureObject=@n8, TextureSize=@n9, WorldPosition=@n10) {materialFunction=&WorldAlignedTexture}.XYZ Texture) {constB=0.5}) {constB=0.5}
  NamedRerouteDeclaration(None=Multiply(A=Add(A=@n6.Result, B=@n7), B=Power(Base=@n11) {constExponent=4})) {name=AnimatedBand, nodeColor=<rgba 0.4297 0 1 1>}
  n12 = PixelNormalWS()
  Power(
   Base=OneMinus(None=Fresnel(Normal=@n12) {baseReflectFraction=0, exponent=1})
   Exp=ScalarParameter() {defaultValue=8, parameterName=Fresnel_Exponent, sliderMax=10}
  ) {constExponent=5}
  n13 = MaterialFunctionCall() {materialFunction=&MF_CharacterParent_2023_PreFX}
  n14 = GetMaterialAttributes(None=@n13.Result) {attributeGetTypes [#1, #2]}
  n15 = TextureObjectParameter() {parameterName=FX, samplerType=SAMPLERTYPE_Masks, texture=/SpireSharedAssets/Assets/Textures/T_Container_Slag_A_FX}
  n16 = TextureSample(Tex=@n15) {samplerType=SAMPLERTYPE_Masks, texture=&T_Container_Slag_A_FX}
  n17 = Multiply(A=TextureSample!(UVs, Tex=@n15, Bias=Constant() {r=4}, Apply View MipBias) {mipValueMode=TMVM_MipBias, samplerType=SAMPLERTYPE_Masks, texture=&T_Container_Slag_A_FX}.R) {constB=0.3333}
  n18 = Divide(A=LocalPosition().Z) {constB=128}
  n19 = OneMinus(None=Saturate(None=Add(A=@n18) {constB=-1}) {desc=thickness})
  n20 = VectorParameter() {parameterName=Emissive_Color, defaultValue=<rgba 1 0.3697 0 20>}
  n21 = Multiply(A=Multiply(A=NamedRerouteUsage(), B=Power(Base=@n19)), B=Multiply(A=@n20.RGB, B=@n20.RGB))
  n22 = LocalPosition()
  n23 = Saturate(None=Divide(A=Length(None=@n22.XY)) {constB=250}) {desc=thickness}
  n24 = Multiply(A=Power(Base=OneMinus(None=@n23)) {constExponent=6}, B=Saturate(None=Divide(A=@n22.Z) {constB=250}) {desc=thickness})
  n25 = Add(A=@n14.Emissive Color, B=Multiply(A=Max(A=@n16.R, B=@n17), B=Multiply(A=@n21, B=@n24)))
  n26 = GetMaterialAttributes(None=MaterialFunctionCall(
    Material=SetMaterialAttributes(MaterialAttributes=@n14.MaterialAttributes, Emissive Color=@n25) {attributeSetTypes [#2]}
    M=@n13.M
    Pre Skinned Local Position=@n13.M
    Distance Blend=@n13.M
    Fresnel=@n13.M
   ) {materialFunction=&MF_CharacterParent_2023_PostFX}.Result) {attributeGetTypes [#3]}
  n27 = MaterialFunctionCall() {materialFunction=&MF_FN_DestructibleDamage}
  n28 = Multiply(A=ScalarParameter() {defaultValue=1, parameterName=HitFXWobble_Scale}, B=@n27.DamageAnimWPO)
  SetMaterialAttributes!(
   MaterialAttributes=@n26.MaterialAttributes
   World Position Offset=Add(A=@n26.World Position Offset, B=StaticSwitchParameter(True=@n28, False=@n27.DamageAnimWPO) {parameterName=HitFXWobble_Scale})
  ) {attributeSetTypes [#3]}
 ]
}
