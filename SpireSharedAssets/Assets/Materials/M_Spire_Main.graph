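# Main lit Spire material, usable on skeletal meshes, that reveals glowing cracks as its per-primitive Health value drops.
# Local-space biplanar noise and crack textures combine with Health (custom primitive data) into a crack mask.
# The mask blends color, crack normals and scalar surface values, while a wider Fresnel-weighted mask shapes HDR glow.
# Time-scrolled bands along normalized height form an animated mask, and a parameter controls glow animation blend.
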Material /SpireSharedAssets/Assets/Materials/M_Spire_Main
graph {
 properties {
  bAutomaticallySetUsageInEditor=false
  bUsedWithMeshDeformer=true
  bUsedWithSkeletalMesh=true
  bUsedWithStaticLighting=true
  usedShadingModels=Default Lit
 }
 nodes [
  n1 = MaterialFunctionCall!(
   ColorTex=TextureObject() {samplerType=SAMPLERTYPE_Grayscale, texture=/Game/TimeOfDay/TODM_Sequence/Materials/Sky/Textures/LowResBlurredNoise_Radish_LF_Root}
   NormalTex
   Pos
   Nrm
   ColorUVScale=ScalarParameter() {defaultValue=400, parameterName=BreakUp_UVScale, sliderMax=1000}
   ColorBlendFactor
   NormalStrength
   NormalUVScale
   NormalBlendFactor
   Local Space Inputs=StaticBool() {value=true}
  ) {materialFunction=/Game/Global/MaterialFunctions/MF_BiPlanar}
  n2 = ScalarParameter() {defaultValue=400, parameterName=Crack_UVScale, sliderMax=1000}
  n3 = Constant() {r=1}
  n4 = MaterialFunctionCall(
   ColorTex=TextureObject() {samplerType=SAMPLERTYPE_Masks, texture=/Game/Athena/WinterArt/Textures/Cracks_05_M}
   NormalTex=TextureObject() {samplerType=SAMPLERTYPE_Normal, texture=/Game/Athena/WinterArt/Textures/Cracks_05_N}
   ColorUVScale=@n2
   ColorBlendFactor=@n3
   NormalUVScale=@n2
   NormalBlendFactor=@n3
   Local Space Inputs=StaticBool() {value=true}
  ) {materialFunction=&MF_BiPlanar}
  n5 = Add(A=Saturate(None=Power(Base=@n1.Color) {constExponent=1.5}), B=Subtract(A=Multiply(A=@n4.Color) {constB=2}))
  n6 = ScalarParameter() {bUseCustomPrimitiveData=true, defaultValue=1, parameterName=Health, sliderMax=1}
  n7 = ComponentMask(None=LinearInterpolate(A=Saturate(None=@n5), Alpha=Saturate(None=@n6)) {constA=1}) {r=true}
  Add(A=@n7) {constB=0.1}
  n8 = Divide(A=WorldPosition().XYZ, B=PreSkinnedLocalBounds().Max)
  n9 = ComponentMask(None=@n8) {b=true}
  Add(A=Multiply(A=@n9) {constB=0.5}) {constB=0.5}
  Add(A=ComponentMask(None=@n8) {b=true, r=true}) {constB=0.5}
  n10 = Subtract(A=Multiply(A=@n9) {constB=0.25}, B=Divide(A=Time()) {constB=1.5})
  Frac(None=@n10)
  Frac(None=@n10)
  LinearInterpolate(A=@n4.Normals, B=Constant3Vector() {constant=<rgba 0 0 1 0>}, Alpha=OneMinus(None=NamedRerouteUsage()))
  n11 = NamedRerouteDeclaration(None=SmoothStep(Value=@n7) {constMax=0.65, constMin=0.75}) {name=CrackMask, nodeColor=<rgba 1 0 0.2109 1>}
  LinearInterpolate(
   A=VectorParameter() {parameterName=Color, defaultValue=<rgba 0.0938 0.0888 0.0854 1>}.RGB
   B=ComponentMask(None=Constant4Vector() {constant=<rgba 0.8177 0.5062 0.1022 1>}.RGBA) {b=true, g=true, r=true}
   Alpha=@n11
  )
  LinearInterpolate(Alpha=@n11) {constA=0.8, constB=0.2}
  n12 = Multiply(A=SmoothStep(Value=@n7) {constMax=0.3, constMin=0.9}, B=OneMinus(None=Fresnel() {baseReflectFraction=0, exponent=1}))
  n13 = VectorParameter() {parameterName=GlowColor, defaultValue=<rgba 1 0.3511 0 30>}
  n14 = Multiply(A=@n13.RGB, B=@n13.RGB)
  LinearInterpolate(B=Multiply(A=Multiply(A=@n12, B=@n14), B=NamedRerouteUsage()), Alpha=ScalarParameter() {parameterName=GlowAnimation_Blend, sliderMax=1})
  n15 = Constant() {r=0.1}
  n16 = Constant()
  n17 = Constant() {r=-0.2}
  n18 = MaterialFunctionCall!(Width=Constant() {r=0.4}, Sharpness=@n16, Offset, Direction Switch, Compare, Input Coordinates=@n10) {materialFunction=/Engine/Functions/Engine_MaterialFunctions03/Procedurals/GeneratedBand}
  NamedRerouteDeclaration(None=Add(
    A=MaterialFunctionCall(Width=@n15, Sharpness=@n16, Offset=@n17, Input Coordinates=@n10) {materialFunction=&GeneratedBand}.Result
    B=Multiply(A=@n18.Result) {constB=0.1}
   )) {name=AnimatedBand, nodeColor=<rgba 0.4297 0 1 1>}
  Power(Base=@n11) {constExponent=0.4545}
 ]
}
