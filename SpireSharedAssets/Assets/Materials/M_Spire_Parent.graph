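# Parent material for Spire assets, layering damage cracks, pulsing weak-point glow, wear and thin-film tint onto MF_CharacterParent_2023_PreFX.
# Custom primitive data supplies Health, weak-point fade and several element toggles.
# World-aligned crack textures sampled at pre-skinned local position, gradients and sphere masks shape damage and emissive.
# Vertex-color element masks form a glow mask, and temporal dither drives opacity for main parts and floaters, before MF_TechArtOverrides.
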Material /SpireSharedAssets/Assets/Materials/M_Spire_Parent
graph {
 properties {
  bAutomaticallySetUsageInEditor=false
  bUseMaterialAttributes=true
  bUsedWithClothing=true
  bUsedWithMeshDeformer=true
  bUsedWithMorphTargets=true
  bUsedWithNanite=true
  bUsedWithNiagaraMeshParticles=true
  bUsedWithSkeletalMesh=true
  shadingModel=MSM_FromMaterialExpression
  usedShadingModels=Default Lit | Subsurface | Clear Coat | Subsurface Profile | Cloth
 }
 nodes [
  n1 = Divide(A=WorldPosition().XYZ, B=PreSkinnedLocalBounds().Max)
  Add(A=ComponentMask(None=@n1) {b=true, r=true}) {constB=0.5}
  n2 = ComponentMask(None=@n1) {b=true}
  Add(A=Multiply(A=@n2) {constB=0.5}) {constB=0.5}
  n3 = TextureObject() {texture=/Game/Global/Textures/Default/Masks_and_Gradients/Cracks_02}
  n4 = ScalarParameter() {defaultValue=200, group=01_Spire, parameterName=Crack_UVScale, sliderMax=1000}
  n5 = VertexInterpolator(VS=LocalPosition() {localOrigin=InstancePreSkinning}.XYZ)
  n6 = Constant() {r=400}
  n7 = Max(
   A=ComponentMask(None=MaterialFunctionCall(TextureObject=@n3, TextureSize=@n4, WorldPosition=@n5.PS) {materialFunction=&WorldAlignedTexture}.XYZ Texture) {r=true}
   B=ComponentMask(None=MaterialFunctionCall(TextureObject=@n3, TextureSize=@n6, WorldPosition=@n5.PS) {materialFunction=&WorldAlignedTexture}.XYZ Texture) {r=true}
  )
  n8 = Multiply(A=Power(Base=NamedRerouteUsage()) {constExponent=9}) {constB=200}
  n9 = Saturate(None=LinearInterpolate(B=OneMinus(None=@n7), Alpha=Saturate(None=@n8)) {constA=1})
  n10 = Multiply(A=NamedRerouteUsage()) {constB=2}
  n11 = Saturate(None=@n10)
  n12 = Subtract(A=@n10)
  n13 = ComponentMask(None=LinearInterpolate(A=LinearInterpolate(B=@n9, Alpha=@n11) {constA=1}, B=Power(Base=@n9) {constExponent=4}, Alpha=Saturate(None=@n12))) {r=true}
  n14 = NamedRerouteDeclaration(None=Multiply(A=NamedRerouteUsage(), B=SmoothStep(Value=@n13) {constMax=0, constMin=0.5})) {name=CrackMask, nodeColor=<rgba 1 0 0.2109 1>}
  n15 = Fresnel() {baseReflectFraction=0, exponent=1}
  n16 = Multiply(A=NamedRerouteUsage(), B=Power(Base=OneMinus(None=@n15)) {constExponent=4})
  n17 = Power(Base=NamedRerouteUsage()) {constExponent=4}
  n18 = Divide(A=Time()) {constB=3.333}
  n19 = Multiply(A=Sine(None=Add(A=@n17, B=@n18))) {constB=0.5}
  Add(A=Multiply(A=Multiply(A=@n14, B=@n16), B=Add(A=@n19) {constB=0.75, desc=crack glow}))
  MaterialExpressionMaterialFunctionCall_19 = MaterialFunctionCall() {materialFunction=&MF_CharacterParent_2023_PreFX}
  n20 = GetMaterialAttributes(None=@MaterialExpressionMaterialFunctionCall_19.Result) {attributeGetTypes [#1, #2, #3]}
  n21 = NamedRerouteUsage()
  n22 = Constant()
  n23 = Multiply(A=NamedRerouteUsage(), B=OneMinus(None=NamedRerouteUsage()))
  n24 = Multiply(A=@n23, B=ScalarParameter() {defaultValue=0.1, group=01_Spire, parameterName=Wear_Intensity, sliderMax=1}) {constB=0.2}
  n25 = SetMaterialAttributes!(
   MaterialAttributes=@n20.MaterialAttributes
   Emissive Color=Add(
    A=@n20.Emissive Color
    B=StaticSwitchParameter(True=@n21, False=@n22) {defaultValue=true, group=01_Spire, parameterName=DamageCracks, sortPriority=0}
   ) {desc=E}
   Base Color=Add(A=Add(A=@n20.Base Color, B=@n24), B=NamedRerouteUsage()) {desc=BC}
   Roughness=Subtract(A=@n20.Roughness, B=Multiply(A=@n23) {constB=0.2})
   Opacity Mask=NamedRerouteUsage()
  ) {attributeSetTypes [
    #1
    #2
    #3
    #4
   ]}
  MaterialFunctionCall(In=@n25) {materialFunction=&MF_TechArtOverrides}
  n26 = Convert(None=@MaterialExpressionMaterialFunctionCall_19.M) {
   nodeName=Break Float3
   convertInputs [{
     type=Vector3
     expressionInput {
      expression=&self:MaterialExpressionMaterialFunctionCall_19
      inputName=None
      mask=0
      maskA=0
      maskB=0
      maskG=0
      maskR=0
      outputIndex=1
     }
    }]
   convertOutputs [{type=Scalar}, {type=Scalar}, {type=Scalar}]
  }
  NamedRerouteDeclaration(None=@n26) {name=M_RChannel, nodeColor=<rgba 0.2891 0 1 1>}
  NamedRerouteDeclaration(None=@n26) {name=M_GChannel, nodeColor=<rgba 0.0312 1 0 1>}
  n27 = SphereMask(
   A=MaterialFunctionCall!(Mask 1=NamedRerouteUsage(), Mask 2=NamedRerouteUsage(), Mask 3=NamedRerouteUsage(), Mask 4=NamedRerouteUsage(), Blend=Constant() {r=3}) {materialFunction=/Game/Effects/Niagara/MasterMaterials/Functions/Utility/MF_Lerp_4_Scalar}.Result
   B=ScalarParameter() {defaultValue=1, group=01_Spire, parameterName=M_SamplePos}
   Radius=ScalarParameter() {defaultValue=1, group=01_Spire, parameterName=M_Radius}
   Hardness=ScalarParameter() {group=01_Spire, parameterName=M_Hardness}
  )
  n28 = Multiply(
   A=TextureCoordinate()
   B=AppendVector(
    A=ScalarParameter() {defaultValue=3, group=01_Spire, parameterName=EmissiveCrack_Tile_X_2}
    B=ScalarParameter() {defaultValue=3, group=01_Spire, parameterName=EmissiveCrack_Tile_Y_2}
   )
  )
  n29 = Multiply(
   A=OneMinus(None=TextureSample(UVs=@n28) {samplerType=SAMPLERTYPE_Masks, texture=/Game/Athena/WinterArt/Textures/Cracks_05_M}.R)
   B=ScalarParameter() {defaultValue=1, group=01_Spire, parameterName=Noise_Strength}
  )
  n30 = Multiply(A=Saturate(None=Subtract(A=@n27, B=@n29)), B=ScalarParameter() {defaultValue=2, group=01_Spire, parameterName=Mask_Str})
  NamedRerouteDeclaration(None=Subtract(A=Saturate(None=@n30), B=NamedRerouteUsage())) {name=DamageMask_01, nodeColor=<rgba 0 0.1562 1 1>}
  n31 = TextureSampleParameter2D() {
   group=00_Texture Inputs
   parameterName=Thickness
   samplerType=SAMPLERTYPE_Masks
   sortPriority=1
   texture=/SpireSharedAssets/Assets/Textures/T_SpireFloat_THICK
  }
  NamedRerouteDeclaration(None=@n31.B) {name=WearMask, nodeColor=<rgba 0 0.1385 1 1>}
  n32 = ScalarParameter() {bUseCustomPrimitiveData=true, defaultValue=1, parameterName=Health, sliderMax=1}
  NamedRerouteDeclaration(None=Saturate(None=OneMinus(None=@n32))) {name=Health, nodeColor=<rgba 1 0.2687 0 1>}
  NamedRerouteDeclaration(None=@n31.R) {name=Thickness, nodeColor=<rgba 1 0 0.0582 1>}
  n33 = TextureSampleParameter2D() {
   group=00_Texture Inputs
   parameterName=FX
   samplerType=SAMPLERTYPE_Masks
   sortPriority=1
   texture=/SpireSharedAssets/Assets/Textures/T_SpireTop_FX
  }
  NamedRerouteDeclaration(None=@n33.R) {name=FX_GoldMask, nodeColor=<rgba 1 0 0.0017 1>}
  n34 = ObjectPositionWS()
  n35 = Add(A=WorldPosition().XYZ)
  n36 = OneMinus()
  n37 = VertexColor()
  n38 = Ceil(None=@n37)
  n39 = Max(A=Ceil(None=@n37.G), B=Ceil(None=@n37.B))
  n40 = OneMinus(None=Multiply(A=Max(A=@n38, B=@n39), B=ScalarParameter() {group=03_Opacity Mask, parameterName=HideElementsOnlyConsidersRedChannel}))
  NamedRerouteDeclaration(None=Multiply(A=Subtract(A=@n34, B=@n35), B=Multiply(A=@n36, B=@n40))) {name=WPO Hidding, nodeColor=<rgba 1 0.9141 0 1>}
  n41 = ScalarParameter() {group=03_Opacity Mask, parameterName=Hide Element 00, primitiveDataIndex=1, sortPriority=0}
  n42 = Constant()
  n43 = VertexColor()
  n44 = LinearInterpolate(
   A=ChannelMaskParameter(None=AppendVector(A=@n43, B=@n43)) {group=03_Opacity Mask, parameterName=Hide Vertex Channel}.R
   B=Max(A=Max(A=@n43.R, B=@n43.R), B=@n43.B)
   Alpha=ScalarParameter() {group=03_Opacity Mask, parameterName=HideMaxChannelColor}
  )
  n45 = ScalarParameter() {bUseCustomPrimitiveData=true, group=03_Opacity Mask, parameterName=Hide Element 01, primitiveDataIndex=1, sortPriority=1}
  n46 = Constant() {r=0.1}
  n47 = ScalarParameter() {bUseCustomPrimitiveData=true, group=03_Opacity Mask, parameterName=Hide Element 02, primitiveDataIndex=2, sortPriority=3}
  n48 = Max(
   A=Max(
    A=MaterialFunctionCall!(Hide Element Toggle=@n41, VertexColorChannel=@n44, Vertex Value to hide=@n42) {materialFunction=/Game/Packages/Fortress_SharedMaterials/Base_Material_Functions/VertexColorMask}.Result
    B=MaterialFunctionCall(Hide Element Toggle=@n45, VertexColorChannel=@n44, Vertex Value to hide=@n46) {materialFunction=&VertexColorMask}.Result
   )
   B=MaterialFunctionCall(Hide Element Toggle=@n47, VertexColorChannel=@n44, Vertex Value to hide=Constant() {r=0.2}) {materialFunction=&VertexColorMask}.Result
  )
  n49 = ScalarParameter() {bUseCustomPrimitiveData=true, group=03_Opacity Mask, parameterName=Hide Element 03, primitiveDataIndex=3, sortPriority=5}
  n50 = MaterialFunctionCall(Hide Element Toggle=@n49, VertexColorChannel=@n44, Vertex Value to hide=Constant() {r=0.3}) {materialFunction=&VertexColorMask}
  n51 = ScalarParameter() {group=03_Opacity Mask, parameterName=Hide Element 04, primitiveDataIndex=4, sortPriority=7}
  n52 = Constant() {r=0.4}
  n53 = ScalarParameter() {bUseCustomPrimitiveData=true, group=03_Opacity Mask, parameterName=Hide Element 05, primitiveDataIndex=4, sortPriority=9}
  n54 = Max(
   A=Max(
    A=Max(A=@n48, B=@n50.Result)
    B=MaterialFunctionCall(Hide Element Toggle=@n51, VertexColorChannel=@n44, Vertex Value to hide=@n52) {materialFunction=&VertexColorMask}.Result
   )
   B=MaterialFunctionCall(Hide Element Toggle=@n53, VertexColorChannel=@n44, Vertex Value to hide=Constant() {r=0.5}) {materialFunction=&VertexColorMask}.Result
  )
  n55 = ScalarParameter() {bUseCustomPrimitiveData=true, group=03_Opacity Mask, parameterName=Hide Element 06, primitiveDataIndex=5, sortPriority=11}
  n56 = MaterialFunctionCall(Hide Element Toggle=@n55, VertexColorChannel=@n44, Vertex Value to hide=Constant() {r=0.6}) {materialFunction=&VertexColorMask}
  n57 = ScalarParameter() {group=03_Opacity Mask, parameterName=Hide Element 07, sortPriority=13}
  n58 = Constant() {r=0.7}
  n59 = Max(
   A=Max(
    A=Max(A=@n54, B=@n56.Result)
    B=MaterialFunctionCall(Hide Element Toggle=@n57, VertexColorChannel=@n44, Vertex Value to hide=@n58) {materialFunction=&VertexColorMask}.Result
   )
   B=MaterialFunctionCall(
    Hide Element Toggle=ScalarParameter() {group=03_Opacity Mask, parameterName=Hide Element 08, sortPriority=15}
    VertexColorChannel=@n44
    Vertex Value to hide=Constant() {r=0.8}
   ) {materialFunction=&VertexColorMask}.Result
  )
  n60 = MaterialFunctionCall(
   Hide Element Toggle=ScalarParameter() {group=03_Opacity Mask, parameterName=Hide Element 09, sortPriority=17}
   VertexColorChannel=@n44
   Vertex Value to hide=Constant() {r=0.9}
  ) {materialFunction=&VertexColorMask}
  n61 = ScalarParameter() {group=03_Opacity Mask, parameterName=Hide Element 10, sortPriority=19}
  n62 = Constant() {r=1}
  NamedRerouteDeclaration(None=Max(
    A=Max(A=@n59, B=@n60.Result)
    B=MaterialFunctionCall(Hide Element Toggle=@n61, VertexColorChannel=@n44, Vertex Value to hide=@n62) {materialFunction=&VertexColorMask}.Result
   )) {name=GlowMask, nodeColor=<rgba 0 1 0.4844 1>}
  n63 = Subtract(A=NamedRerouteUsage(), B=NamedRerouteUsage())
  NamedRerouteDeclaration(None=@n63) {name=Emissive_Mask, nodeColor=<rgba 0 1 0.2969 1>}
  NamedRerouteDeclaration(None=@MaterialExpressionMaterialFunctionCall_19.Pre skinned Local Position) {name=PreSkinnedLocalPosition, nodeColor=<rgba 0.3828 1 0 1>}
  n64 = Multiply(A=TextureCoordinate(), B=ScalarParameter() {defaultValue=0.1, group=01_Spire, parameterName=TextureVartiation_Scaling, sliderMax=1}) {constB=0.1}
  n65 = Add(
   A=TextureSample(UVs=Multiply(A=@n64) {constB=1.333}) {samplerType=SAMPLERTYPE_LinearGrayscale, texture=/Engine/EngineMaterials/Good64x64TilingNoiseHighFreq}.R
   B=TextureSample(UVs=Multiply(A=@n64) {constB=2.666}) {samplerType=SAMPLERTYPE_LinearGrayscale, texture=&Good64x64TilingNoiseHighFreq}.R
  )
  n66 = TextureSample(UVs=Multiply(A=@n64) {constB=4.222}) {samplerType=SAMPLERTYPE_LinearGrayscale, texture=&Good64x64TilingNoiseHighFreq}
  NamedRerouteDeclaration(None=Divide(A=Add(A=@n65, B=@n66.R)) {constB=3}) {name=TextureVariation, nodeColor=<rgba 1 0.1172 0 1>}
  n67 = Add(A=NamedRerouteUsage(), B=VectorParameter() {parameterName=DamageGradientCenter, defaultValue=<rgba 0 0 -1000 1>}.RGB)
  n68 = ScalarParameter() {defaultValue=1000, group=01_Spire, parameterName=GradientLength}
  n69 = Abs(None=ChannelMaskParameter(None=Divide(A=@n67, B=@n68)) {maskChannel=Blue, parameterName=DamageGradientDirection, defaultValue=<rgba 0 0 1 0>}.R)
  n70 = Add(A=Saturate(None=@n69))
  n71 = Multiply(A=NamedRerouteUsage()) {constB=2}
  NamedRerouteDeclaration(None=Saturate(None=Subtract(A=@n70, B=@n71))) {name=Damage_TransitionGradient, nodeColor=<rgba 0.875 0 1 1>}
  n72 = OneMinus(None=Fresnel(Normal=PixelNormalWS()) {baseReflectFraction=0, exponent=1})
  n73 = NamedRerouteUsage()
  n74 = NamedRerouteUsage()
  n75 = NamedRerouteUsage()
  n76 = NamedRerouteUsage()
  n77 = Constant()
  n78 = Multiply(
   A=SphereMask(
    A=MaterialFunctionCall(Mask 1=@n73, Mask 2=@n74, Mask 3=@n75, Mask 4=@n76, Blend=@n77) {materialFunction=&MF_Lerp_4_Scalar}.Result
    B=ScalarParameter() {group=01_Spire, parameterName=E2_SamplePos}
    Radius=ScalarParameter() {defaultValue=1, group=01_Spire, parameterName=E2_Radius}
    Hardness=ScalarParameter() {group=01_Spire, parameterName=E2_Hardness}
   )
   B=ScalarParameter() {defaultValue=1.2, group=01_Spire, parameterName=AO_Stregnth}
  ) {constB=2}
  n79 = NamedRerouteUsage()
  n80 = ScalarParameter() {defaultValue=1, group=01_Spire, parameterName=Z-Gradient_SamplePos}
  n81 = ScalarParameter() {defaultValue=0.8, group=01_Spire, parameterName=Z-Gradient_SampleRaidus}
  n82 = ScalarParameter() {defaultValue=0.5, group=01_Spire, parameterName=Z-Gradient_SampleHardness}
  n83 = Saturate(None=Subtract(A=Saturate(None=@n78), B=SphereMask(A=@n79, B=@n80, Radius=@n81, Hardness=@n82)))
  n84 = NamedRerouteUsage()
  n85 = ScalarParameter() {defaultValue=0.6, group=01_Spire, parameterName=Z-Gradient_SamplePos_02}
  n86 = ScalarParameter() {defaultValue=0.3, group=01_Spire, parameterName=Z-Gradient_SampleRaidus_02}
  n87 = ScalarParameter() {defaultValue=0.25, group=01_Spire, parameterName=Z-Gradient_SampleHardness_02}
  n88 = ScalarParameter() {defaultValue=3, group=01_Spire, parameterName=EmissiveCrack_Tile_X}
  n89 = ScalarParameter() {defaultValue=3, group=01_Spire, parameterName=EmissiveCrack_Tile_Y}
  n90 = TextureSample(UVs=Multiply(A=TextureCoordinate(), B=AppendVector(A=@n88, B=@n89))) {samplerType=SAMPLERTYPE_Masks, texture=&Cracks_05_M}
  n91 = ScalarParameter() {defaultValue=1, group=01_Spire, parameterName=EmissiveNoise_Str}
  n92 = Saturate(None=Subtract(A=SphereMask(A=@n84, B=@n85, Radius=@n86, Hardness=@n87), B=Multiply(A=@n90.R, B=@n91)))
  NamedRerouteDeclaration(None=Multiply(A=Power(Base=@n72), B=Max(A=@n83, B=@n92))) {name=Damage_EmissiveCracks, nodeColor=<rgba 0 0.5312 1 1>}
  NamedRerouteDeclaration(None=@n31.G) {name=DeathMask, nodeColor=<rgba 0.1693 1 0 1>}
  n93 = LinearInterpolate(
   A=ScalarParameter() {group=01_Spire, parameterName=OpacityMask_MainParts, sliderMax=1}
   B=ScalarParameter() {group=01_Spire, parameterName=OpacityMask_Floaters, sliderMax=1}
   Alpha=NamedRerouteUsage()
  )
  NamedRerouteDeclaration(None=MaterialFunctionCall!(Alpha Threshold=OneMinus(None=@n93), Random) {materialFunction=/Game/TimeOfDay/CloudsAndStorm/Materials/MaterialFunctions/DitherTemporalAABlue}.Result) {name=OpacityMask, nodeColor=<rgba 0.8516 1 0 1>}
  n94 = CurveAtlasRowParameter(CurveTime=Fresnel(ExponentIn=ScalarParameter() {defaultValue=1, group=01_Spire, parameterName=Anodizing_Exponent, sliderMax=32}) {baseReflectFraction=0, exponent=3}) {
   atlas=&/CA_SpireCurves
   curve=&/CC_SpireThinFilm
   group=01_Spire
   parameterName=Anodizing Color
  }
  n95 = ScalarParameter() {defaultValue=0.03, group=01_Spire, parameterName=Anodizing, sliderMax=1}
  n96 = NamedRerouteUsage()
  NamedRerouteDeclaration(None=Multiply(A=Multiply(A=@n94, B=@n95), B=OneMinus(None=@n96))) {name=ThinFilm, nodeColor=<rgba 0 1 0.5606 1>}
  n97 = Multiply(
   A=Multiply(A=NamedRerouteUsage(), B=VectorParameter() {parameterName=CrackGlowColor, defaultValue=<rgba 0.7913 0.305 0 1>}.RGB)
   B=ScalarParameter() {defaultValue=3, parameterName=CrackGlowStrength}
  )
  n98 = Power(Base=VertexColor().G, Exp=ScalarParameter() {defaultValue=4, group=01_Spire, parameterName=Distance_Exponent, sliderMax=8}) {constExponent=4}
  n99 = Add(A=Multiply(A=NamedRerouteUsage(), B=@n98), B=@n98)
  n100 = Add(
   A=Multiply(A=NamedRerouteUsage(), B=ScalarParameter() {defaultValue=2, group=01_Spire, parameterName=Mask_Glow_Strenght, sliderMax=4}) {constB=0.5}
   B=NamedRerouteUsage()
  )
  n101 = Time()
  n102 = Multiply(A=Sine(None=Divide(A=@n101) {constB=1.2})) {constB=0.5}
  n103 = ScalarParameter() {bUseCustomPrimitiveData=true, defaultValue=1, group=03_Opacity Mask, parameterName=WeakPointFade, primitiveDataIndex=6, sortPriority=11}
  n104 = Multiply(A=Multiply(A=Multiply(A=@n99, B=@n100) {constA=1}, B=NamedRerouteUsage()) {constA=1}, B=Multiply(A=Add(A=@n102) {constB=0.65}, B=@n103)) {constA=1, desc=weakpoint glow}
  n105 = VectorParameter() {group=01_Spire, parameterName=Glow_ColorA, defaultValue=<rgba 0.799 0.3098 0 5>}
  n106 = VectorParameter() {group=01_Spire, parameterName=Glow_ColorB, defaultValue=<rgba 1 0.7558 0.4719 5>}
  n107 = LinearInterpolate(A=Multiply(A=@n105.RGB, B=@n105.RGB), B=Multiply(A=@n106.RGB, B=@n106.RGB), Alpha=Saturate(None=Power(Base=@n104) {constExponent=4}))
  NamedRerouteDeclaration(None=Add(A=Power(Base=@n97) {constExponent=1.5}, B=Multiply(A=@n104, B=@n107))) {name=Emisive, nodeColor=<rgba 1 0 0.439 1>}
  NamedRerouteDeclaration(None=@n33.G) {name=FX_StoneMask, nodeColor=<rgba 0 1 0 1>}
  NamedRerouteDeclaration(None=@n26) {name=M_BChannel, nodeColor=<rgba 0 1 0.2734 1>}
  NamedRerouteDeclaration(None=SetMaterialAttributes!(MaterialAttributes, Base Color=@n63, Emissive Color=@n63) {attributeSetTypes [#2, #1]}) {name=DebugMat, nodeColor=<rgba 0.7344 0 1 1>}
  NamedRerouteDeclaration(None=@n69) {name=Z-GradientMask, nodeColor=<rgba 1 0 0.7734 1>}
  n108 = Constant() {r=0.1}
  n109 = Constant()
  n110 = Constant() {r=-0.2}
  n111 = Subtract(A=Multiply(A=@n2) {constB=0.25}, B=Divide(A=Time()) {constB=2.8})
  n112 = MaterialFunctionCall!(Width=Constant() {r=0.4}, Sharpness=@n109, Offset, Direction Switch, Compare, Input Coordinates=@n111) {materialFunction=/Engine/Functions/Engine_MaterialFunctions03/Procedurals/GeneratedBand}
  NamedRerouteDeclaration(None=Add(
    A=MaterialFunctionCall(Width=@n108, Sharpness=@n109, Offset=@n110, Input Coordinates=@n111) {materialFunction=&GeneratedBand}.Result
    B=Multiply(A=@n112.Result) {constB=0.1}
   )) {name=AnimatedBand, nodeColor=<rgba 0.4297 0 1 1>}
 ]
}
