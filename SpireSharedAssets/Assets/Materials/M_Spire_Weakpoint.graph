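# Additive, Nanite- and skeletal-capable material for a Spire weak point, drawing a pulsing colored glow.
# A time-driven sine pushes vertices along their normals, alongside a Fresnel with depth fade and a squared color parameter with intensity in alpha.
# A vertex-color mask with eleven Hide Element toggles drives a world position offset toward the object pivot to hide mesh parts.
# A panning, distorted-noise branch also builds a two-color blend, faded opacity and CurlWPO displacement.
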Material /SpireSharedAssets/Assets/Materials/M_Spire_Weakpoint
graph {
 properties {
  bAutomaticallySetUsageInEditor=false
  bUsedWithMeshDeformer=true
  bUsedWithNanite=true
  bUsedWithSkeletalMesh=true
  bUsedWithStaticLighting=true
  blendMode=BLEND_Additive
  usedShadingModels=Default Lit
 }
 nodes [
  n1 = Sine(None=Divide(A=Time(), B=ScalarParameter() {defaultValue=0.8, group=00 Spire Weak Point, parameterName=Animation_Frequency, sliderMax=4, sortPriority=31}) {constB=0.888})
  n2 = Multiply(
   A=Add(A=Multiply(A=@n1) {constB=0.5}) {constB=0.5}
   B=ScalarParameter() {defaultValue=5, group=00 Spire Weak Point, parameterName=WPO_Strength, sliderMax=50}
  ) {constB=5}
  n3 = ScalarParameter() {group=00 Spire Weak Point, parameterName=WPO_Offset, sliderMax=50, sortPriority=33}
  Add(A=Multiply(A=VertexNormalWS(), B=Add(A=@n2, B=@n3)) {constB=5}, B=NamedRerouteUsage())
  n4 = ScalarParameter() {defaultValue=0.04, group=00 Spire Weak Point, parameterName=Fresnel_BaseReflectFraction, sliderMax=1, sortPriority=22}
  DepthFade(Opacity=Fresnel(
    ExponentIn=ScalarParameter() {defaultValue=5, group=00 Spire Weak Point, parameterName=Fresnel_Exponent, sliderMax=12, sortPriority=21}
    BaseReflectFractionIn=@n4
   )) {fadeDistanceDefault=15}
  n5 = Divide(A=WorldPosition().XYZ, B=PreSkinnedLocalBounds().Max)
  n6 = Add(A=ComponentMask(None=@n5) {b=true, r=true}) {constB=0.5}
  n7 = TextureSample(UVs=Panner(Coordinate=@n6) {speedX=0.2, speedY=0.1})
  n8 = ScalarParameter() {defaultValue=2, parameterName=NoiseSIze, sliderMax=4}
  n9 = MaterialFunctionCall(Float2=Panner(Coordinate=Divide(A=@n6, B=@n8)) {speedX=0.02, speedY=-0.1}) {materialFunction=&BreakOutFloat2Components}
  n10 = Add(A=Multiply(A=Add(A=@n7.R) {constB=-0.5}, B=ScalarParameter() {defaultValue=0.1, parameterName=NoiseUVOffset, sliderMax=0.5}), B=@n9.R)
  n11 = ComponentMask(None=@n5) {b=true}
  n12 = Multiply(
   A=TextureSample(UVs=MaterialFunctionCall(X=@n10, Y=@n9.G) {materialFunction=&MakeFloat2}.Result) {samplerType=SAMPLERTYPE_LinearColor}.R
   B=@n11
  )
  DepthFade(
   Opacity=Multiply(A=@n12, B=ScalarParameter() {defaultValue=0.5, parameterName=OpacityScale, sliderMax=1}) {constB=0.8}
   FadeDistance=ScalarParameter() {defaultValue=50, parameterName=DepthFade, sliderMax=100}
  ) {desc=Opacity}
  n13 = VectorParameter() {parameterName=ColorA, defaultValue=<rgba 0 1 0.2152 3>}
  n14 = VectorParameter() {parameterName=ColorB, defaultValue=<rgba 0 0.384 1 5>}
  LinearInterpolate(A=Multiply(A=@n13.RGB, B=@n13.RGB), B=Multiply(A=@n14.RGB, B=@n14.RGB), Alpha=@n12) {desc=BaseColor}
  n15 = VectorParameter() {
   group=00 Spire Weak Point
   parameterName=Color
   sortPriority=10
   channelNames {a=Emissive Intensity, b=Color B, g=Color G, r=Color R}
   defaultValue=<rgba 1 0.05 0 10>
  }
  Multiply(A=@n15, B=@n15)
  n16 = ObjectPositionWS()
  n17 = Add(A=WorldPosition().XYZ)
  n18 = ScalarParameter() {group=03_Opacity Mask, parameterName=Hide Element 00, sortPriority=0}
  n19 = Constant()
  n20 = VertexColor()
  n21 = LinearInterpolate(
   A=ChannelMaskParameter(None=AppendVector(A=@n20, B=@n20)) {group=03_Opacity Mask, parameterName=Hide Vertex Channel}.R
   B=Max(A=Max(A=@n20.R, B=@n20.R), B=@n20.B)
   Alpha=ScalarParameter() {group=03_Opacity Mask, parameterName=HideMaxChannelColor}
  )
  n22 = ScalarParameter() {group=03_Opacity Mask, parameterName=Hide Element 01, sortPriority=1}
  n23 = Constant() {r=0.1}
  n24 = Max(
   A=Max(
    A=MaterialFunctionCall!(Hide Element Toggle=@n18, VertexColorChannel=@n21, Vertex Value to hide=@n19) {materialFunction=/Game/Packages/Fortress_SharedMaterials/Base_Material_Functions/VertexColorMask}.Result
    B=MaterialFunctionCall(Hide Element Toggle=@n22, VertexColorChannel=@n21, Vertex Value to hide=@n23) {materialFunction=&VertexColorMask}.Result
   )
   B=MaterialFunctionCall(
    Hide Element Toggle=ScalarParameter() {group=03_Opacity Mask, parameterName=Hide Element 02, sortPriority=3}
    VertexColorChannel=@n21
    Vertex Value to hide=Constant() {r=0.2}
   ) {materialFunction=&VertexColorMask}.Result
  )
  n25 = MaterialFunctionCall(
   Hide Element Toggle=ScalarParameter() {group=03_Opacity Mask, parameterName=Hide Element 03, sortPriority=5}
   VertexColorChannel=@n21
   Vertex Value to hide=Constant() {r=0.3}
  ) {materialFunction=&VertexColorMask}
  n26 = ScalarParameter() {group=03_Opacity Mask, parameterName=Hide Element 04, sortPriority=7}
  n27 = Constant() {r=0.4}
  n28 = Max(
   A=Max(
    A=Max(A=@n24, B=@n25.Result)
    B=MaterialFunctionCall(Hide Element Toggle=@n26, VertexColorChannel=@n21, Vertex Value to hide=@n27) {materialFunction=&VertexColorMask}.Result
   )
   B=MaterialFunctionCall(
    Hide Element Toggle=ScalarParameter() {group=03_Opacity Mask, parameterName=Hide Element 05, sortPriority=9}
    VertexColorChannel=@n21
    Vertex Value to hide=Constant() {r=0.5}
   ) {materialFunction=&VertexColorMask}.Result
  )
  n29 = MaterialFunctionCall(
   Hide Element Toggle=ScalarParameter() {group=03_Opacity Mask, parameterName=Hide Element 06, sortPriority=11}
   VertexColorChannel=@n21
   Vertex Value to hide=Constant() {r=0.6}
  ) {materialFunction=&VertexColorMask}
  n30 = ScalarParameter() {group=03_Opacity Mask, parameterName=Hide Element 07, sortPriority=13}
  n31 = Constant() {r=0.7}
  n32 = Max(
   A=Max(
    A=Max(A=@n28, B=@n29.Result)
    B=MaterialFunctionCall(Hide Element Toggle=@n30, VertexColorChannel=@n21, Vertex Value to hide=@n31) {materialFunction=&VertexColorMask}.Result
   )
   B=MaterialFunctionCall(
    Hide Element Toggle=ScalarParameter() {group=03_Opacity Mask, parameterName=Hide Element 08, sortPriority=15}
    VertexColorChannel=@n21
    Vertex Value to hide=Constant() {r=0.8}
   ) {materialFunction=&VertexColorMask}.Result
  )
  n33 = MaterialFunctionCall(
   Hide Element Toggle=ScalarParameter() {group=03_Opacity Mask, parameterName=Hide Element 09, sortPriority=17}
   VertexColorChannel=@n21
   Vertex Value to hide=Constant() {r=0.9}
  ) {materialFunction=&VertexColorMask}
  n34 = ScalarParameter() {group=03_Opacity Mask, parameterName=Hide Element 10, sortPriority=19}
  n35 = Constant() {r=1}
  n36 = OneMinus(None=Max(
    A=Max(A=@n32, B=@n33.Result)
    B=MaterialFunctionCall(Hide Element Toggle=@n34, VertexColorChannel=@n21, Vertex Value to hide=@n35) {materialFunction=&VertexColorMask}.Result
   ))
  n37 = VertexColor()
  n38 = Ceil(None=@n37)
  n39 = Max(A=Ceil(None=@n37.G), B=Ceil(None=@n37.B))
  n40 = OneMinus(None=Multiply(A=Max(A=@n38, B=@n39), B=ScalarParameter() {group=03_Opacity Mask, parameterName=HideElementsOnlyConsidersRedChannel}))
  NamedRerouteDeclaration(None=Multiply(A=Subtract(A=@n16, B=@n17), B=Multiply(A=@n36, B=@n40))) {name=WPO Hidding, nodeColor=<rgba 1 0.9141 0 1>}
  n41 = Multiply(
   A=ScalarParameter() {defaultValue=0.5, parameterName=NormalOffset, sliderMax=2}
   B=Transform(None=Constant3Vector() {constant=<rgba 0 0 1 0>})
  )
  n42 = Saturate(None=DotProduct(A=VertexNormalWS(), B=Constant3Vector() {constant=<rgba 0 0 1 0>}))
  n43 = Multiply(A=Multiply(A=@n12, B=Multiply(A=@n11, B=@n42)), B=Constant3Vector() {constant=<rgba 0 0 1 0>})
  n44 = ScalarParameter() {defaultValue=2, parameterName=HeightOffset, sliderMax=20}
  NamedRerouteDeclaration(None=Add(A=Multiply(A=@n12, B=@n41), B=Multiply(A=@n43, B=@n44)) {desc=WPO}) {name=CurlWPO, nodeColor=<rgba 0 1 0.9742 1>}
  NamedRerouteUsage()
 ]
}
