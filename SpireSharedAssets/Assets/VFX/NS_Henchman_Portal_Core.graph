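# NS_Henchman_Portal_Core is the Niagara effect for the glowing pink core of a henchman portal.
# SmokeLoopIdle grows core meshes with a pulsing light and mesh-sampled, inward-pulled smoke, scaling colors by System.CurveDestroyed for the destroyed state.
# Absorb_Sparkles pulls sparkles inward, while ChromaticRefraction adds a delayed, expanding distortion sprite.
# Twin CPU and GPU emitters launch fast, heavily dragged, colliding smoke, gating size and alpha on velocity and adjusting alpha by time of day.
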NiagaraSystem /SpireSharedAssets/Assets/VFX/NS_Henchman_Portal_Core
dependencies {
 usedDataInterfaces [
  &NiagaraDataInterfaceColorCurve
  &NiagaraDataInterfaceVectorCurve
  &NiagaraDataInterfaceCurve
  &NiagaraDataInterfaceStaticMesh
  &NiagaraDataInterfaceSpriteRendererInfo
  &NiagaraDataInterfaceArrayFloat
 ]
 usedDynamicInputs []
 usedModules [
  &SystemState
  &self:SystemSpawnScript.SystemScriptSource.SystemScriptGraph.NiagaraNodeAssignment_0.SetVariables_#1
 ]
}
emitters {
 Absorb_Sparkles {
  emitterUpdateScript [
   &EmitterState(Life Cycle Mode=::NewEnumerator1, Loop Duration=2, UseLoopDelay=false)
   &SpawnRate(SpawnRate=10, Use Spawn Probability=false)
  ]
  particleSpawnScript [
   &InitializeParticle(
    Color Maximum=<rgba 3 0.06 0.942 1>
    Color Minimum=<rgba 0.6 0.006 0.1842 1>
    Color Mode=::NewEnumerator2
    Lifetime Max=1.7
    Lifetime Mode=::NewEnumerator1
    Mass Mode=::NewEnumerator3
    Sprite Rotation Mode=::NewEnumerator2
    Sprite Size Mode=::NewEnumerator1
    Uniform Sprite Size Max=19
    Uniform Sprite Size Min=12
    UsePositionOffset=false
    Write Material Random=true
   )
   &ShapeLocation(
    MaskSpawnGroup=false
    Non Uniform Scale=<1 1 1.7>
    Override Seed=false
    Sphere Radius=&RandomRangeFloat(Minimum=100, Maximum=200)
    Sphere Surface Distribution=0.4643
    UseOwnerScale=true
   )
  ]
  particleUpdateScript [
   &ParticleState()
   &ScaleSpriteSize(Uniform Curve Sprite Scale=DataInterface() {propertyValues {Curve {keys [
        {arriveTangent=5.4903, interpMode=RCIM_Cubic, leaveTangent=5.4903, tangentMode=RCTM_User, time=0, value=0}
        {interpMode=RCIM_Cubic, tangentMode=RCTM_User, time=0.4, value=1}
        {interpMode=RCIM_Cubic, tangentMode=RCTM_User, time=1, value=0}
       ]}}})
   &Drag(Use Linear Drag=true, Use Rotational Drag=false) {enabled=false}
   &PointAttractionForce(
    Attraction Radius=400
    Attraction Strength=&RandomRangeFloat(Minimum=50, Maximum=80)
    Falloff Exponent=0.1
    Kill Within Radius=false
    Module.Use Falloff=true
    UseAttractorOffset=false
   )
   &SolveForcesAndVelocity(Clamp Velocity=false, Limit Acceleration=false)
   &ScaleSpriteSizeBySpeed(Max Scale Factor=<0.35 2.5>, Min Scale Factor=<1.2 0.8>, Velocity Threshold=170)
  ]
  rendererClasses [&NiagaraSpriteRendererProperties]
  renderers [{rendererClass=&NiagaraSpriteRendererProperties}]
 }
 CPU_Coll_Smoke001 {
  emitterUpdateScript [
   &EmitterState(Life Cycle Mode=::NewEnumerator1, Loop Behavior=::NewEnumerator1, Loop Delay=0.7, Loop Duration Mode=::NewEnumerator1, UseLoopDelay=true)
   &SpawnRate(SpawnRate=4, Use Spawn Probability=false)
  ]
  particleSpawnScript [
   &InitializeParticle(
    Color=<rgba 0.9759 1 0.8912 2>
    Lifetime=0.9
    Sprite Rotation Mode=::NewEnumerator2
    Sprite Size Mode=::NewEnumerator1
    Uniform Sprite Size Max=150
    Uniform Sprite Size Min=100
    UsePositionOffset=false
    Write Material Random=true
   )
   &AddVelocity(
    Module.Use Falloff Distance=false
    Origin Offset=<0 0 2500>
    Origin Offset Coordinate Space=::World
    Override Seed=false
    Velocity Mode=::NewEnumerator1
    Velocity Speed=7000
   )
   &SubUVAnimation(
    Evaluation Type=::NewEnumerator0
    "Sprite Renderer "=DataInterface() {propertyValues {SpriteRenderer=&self:CPU_Coll_Smoke001_2.NiagaraSpriteRendererProperties_0}}
    SubUV Animation Mode:ENiagara_SubUVLookupModeV2=::NewEnumerator2
    UseEndFrame=false
    UseStartFrame=false
   ) {enabled=false}
   &ShapeLocation(
    MaskSpawnGroup=false
    Offset=<0 0 60>
    Offset Mode=::NewEnumerator0
    Override Seed=false
    Ring Radius=5
    Shape Primitive=::NewEnumerator4
    UseOwnerScale=true
   )
  ]
  particleUpdateScript [
   &ParticleState()
   &CameraOffset(Camera Offset Amount=100)
   &SpriteRotationRate(Rotation Rate=&Multiply_Float(
     A=&RandomRangeFloat(Minimum=-60, Maximum=60)
     B=&FloatFromCurve(
      FloatCurve=DataInterface() {propertyValues {Curve {keys [
          {interpMode=RCIM_Cubic, tangentMode=RCTM_User, time=0.1, value=0}
          {arriveTangent=2.1084, interpMode=RCIM_Cubic, leaveTangent=2.1084, tangentMode=RCTM_User, time=0.7, value=1}
         ]}}}
      Scale Curve=0.01
     )
    ))
   &ScaleSpriteSize(
    Uniform Curve Scale=&MakeCustomFloatFromBool(
     Bool=&SetBoolByFloatComparison(A=0.0001, B=HlslExpression() {hlslExpression="abs(Particles.Velocity.x + Particles.Velocity.y + Particles.Velocity.z)"})
     True Float=1.5
     False Float=0.01
    )
    Uniform Curve Sprite Scale=DataInterface() {propertyValues {Curve {keys [
        {arriveTangent=1.0284, interpMode=RCIM_Cubic, leaveTangent=1.0284, tangentMode=RCTM_User, time=0.01, value=0.1}
        {arriveTangent=1.0246, interpMode=RCIM_Cubic, leaveTangent=1.0246, tangentMode=RCTM_User, time=1, value=1.25}
       ]}}}
   )
   &Collision(
    Age Colliding Particles=false
    Collision Enabled=&SetBoolByFloatComparison(
     A=1
     B=&FloatFromCurve(FloatCurve=DataInterface() {propertyValues {Curve {keys [<at 0.1 0.95>,<at 0.5 2>]}}}, Scale Curve=0.5)
    )
    Enable Rest State=false
    Friction=5
    Friction During a Bounce=5
    Module.EnableMaxCollisionCount=false
    Module.Randomize Collision Normal Vector=false
    Particle Radius Scale=0
    Randomize Collision Normal=0
    Restitution=0
    Simple Friction=false
   )
   &Drag(Drag=20, Rotational Drag=-5, Use Linear Drag=true, Use Rotational Drag=false)
   &SolveForcesAndVelocity(Clamp Velocity=false, Limit Acceleration=false)
   &DynamicMaterialParameters(
    Index 0 Param 1=&RandomRangeFloat(Minimum=0, Maximum=1)
    Index 1 Param 1=&FloatFromCurve(FloatCurve=DataInterface() {propertyValues {Curve {keys [
         {arriveTangent=1.117, interpMode=RCIM_Cubic, leaveTangent=1.117, tangentMode=RCTM_User, time=0, value=0.1}
         {arriveTangent=0.0224, interpMode=RCIM_Cubic, leaveTangent=0.0224, tangentMode=RCTM_User, time=1, value=1}
        ]}}})
    Index 1 Param 3=&FloatFromCurve(
     FloatCurve=DataInterface() {propertyValues {Curve {keys [
         {arriveTangent=1.117, interpMode=RCIM_Cubic, leaveTangent=1.117, tangentMode=RCTM_User, time=0, value=0}
         {arriveTangent=0.0224, interpMode=RCIM_Cubic, leaveTangent=0.0224, tangentMode=RCTM_User, time=1, value=1}
        ]}}}
     Scale Curve=&RandomRangeFloat(Minimum=0.1, Maximum=0.2)
    )
    Module.Param0WriteEnabled=true
    Module.Param0WriteEnabled1=true
    Module.Param1WriteEnabled=false
    Module.Param1WriteEnabled1=true
    Module.Param2WriteEnabled=false
    Module.Param2WriteEnabled1=true
    Module.Param3WriteEnabled=false
    Module.Param3WriteEnabled1=true
    Write Parameter Index 1=true
   )
   &ScaleColor(
    Scale Alpha=&MakeCustomFloatFromBool(
     Bool=&SetBoolByFloatComparison(A=0.001, B=HlslExpression() {hlslExpression="abs(Particles.Velocity.x + Particles.Velocity.y + Particles.Velocity.z)"})
     True Float=&FloatFromCurve(
      FloatCurve=DataInterface() {propertyValues {Curve {keys [
          {tangentMode=RCTM_User, time=0, value=0}
          {interpMode=RCIM_Cubic, tangentMode=RCTM_User, time=0.1, value=0.15}
          {arriveTangent=-0.0405, interpMode=RCIM_Cubic, leaveTangent=-0.0405, tangentMode=RCTM_User, time=0.3, value=1}
          {arriveTangent=-0.3723, interpMode=RCIM_Cubic, leaveTangent=-0.3723, tangentMode=RCTM_User, time=1, value=0.05}
         ]}}}
      Scale Curve=&FNTimeOfDayScalar(Night Scale=0.6, Morning Scale=0.7, Evening Scale=0.7)
     )
     False Float=0
    )
    ScaleA=true
    ScaleRGB=true
   )
  ]
  rendererClasses [&NiagaraSpriteRendererProperties]
  renderers [{rendererClass=&NiagaraSpriteRendererProperties}]
 }
 ChromaticRefraction {
  emitterUpdateScript [
   &EmitterState(
    Inactive Response=::NewEnumerator1
    Life Cycle Mode=::NewEnumerator1
    Loop Behavior=::NewEnumerator1
    Loop Delay=0.5
    Loop Duration Mode=::NewEnumerator1
    UseLoopDelay=true
   )
   &SpawnBurst_Instantaneous(Use Spawn Probability=false, UseLoopCountLimit=false)
  ]
  particleSpawnScript [&InitializeParticle(Sprite Size=<5 8>, Sprite Size Mode=::NewEnumerator4, UsePositionOffset=false, Write Material Random=true)]
  particleUpdateScript [
   &ParticleState(Kill Particles When Lifetime Has Elapsed=false, Let Infinitely Lived Particles Die When Emitter Deactivates=true)
   &Timeline(
    Play Rate=&FloatFromCurve(
     FloatCurve=DataInterface() {propertyValues {Curve {keys [
         {interpMode=RCIM_Cubic, leaveTangent=-12.5334, tangentMode=RCTM_Break, time=0, value=4}
         {interpMode=RCIM_Cubic, tangentMode=RCTM_Break, time=0.7, value=1}
        ]}}}
     Scale Curve=0.5
    )
    Timeline End=1.0000000200408773e+20
   )
   &ScaleSpriteSize(
    Scale Sprite Size Mode=::NewEnumerator0
    Uniform Scale Factor=&FloatFromCurve(
     FloatCurve=DataInterface() {propertyValues {Curve {keys [{interpMode=RCIM_Cubic, tangentMode=RCTM_User, time=0.5, value=0}, {interpMode=RCIM_Cubic, tangentMode=RCTM_User, time=1, value=1}]}}}
     CurveIndex=$Emitter.Age
     Scale Curve=70
    )
   )
   &DynamicMaterialParameters(
    Index 0 Param 1=1
    Index 1 Param 2=500
    Index 1 Param 4=1
    Index 2 Param 1=$Particles.Timeline.Time
    Index 2 Param 2=$Particles.Timeline.Time
    Index 2 Param 3=$Particles.Timeline.Time
    Index 2 Param 4=&SelectFloatFromArray(Direct Array Index=$Particles.PartitionParticles.Partition, Float Selection Array=DataInterface() {propertyValues {FloatData [1,0.3,0.3]}})
    Index 3 Param 1=&FloatFromCurve(FloatCurve=DataInterface() {propertyValues {Curve {keys [<at 0.1 1>,<at 0.4 0>]}}})
    Index 3 Param 2=200
    Module.Index0Param0Write=true
    Module.Index0Param1Write=true
    Module.Index0Param2Write=true
    Module.Index0Param3Write=true
    Module.Index1Param0Write=true
    Module.Index1Param1Write=true
    Module.Index1Param2Write=true
    Module.Index1Param3Write=true
    Module.Index2Param0Write=true
    Module.Index2Param1Write=true
    Module.Index2Param2Write=true
    Module.Index2Param3Write=true
    Module.Index3Param0Write=true
    Module.Index3Param1Write=true
    Module.Index3Param2Write=true
    Module.Index3Param3Write=true
    Write Parameter Index 1=true
    Write Parameter Index 2=true
    Write Parameter Index 3=true
   ) {module=DynamicMaterialParameters001}
   &ScaleColor(Scale RGB=&VectorFromFloat(Value=0.5), ScaleA=true, ScaleRGB=true) {module=ScaleColor001}
  ]
  rendererClasses [&NiagaraSpriteRendererProperties]
  renderers [{rendererClass=&NiagaraSpriteRendererProperties}]
 }
 GPU_Coll_Smoke {
  simTarget=GPUComputeSim
  emitterUpdateScript [
   &EmitterState(Life Cycle Mode=::NewEnumerator1, Loop Behavior=::NewEnumerator1, Loop Delay=0.7, Loop Duration Mode=::NewEnumerator1, UseLoopDelay=true)
   &SpawnRate(SpawnRate=15, Use Spawn Probability=false)
  ]
  particleSpawnScript [
   &InitializeParticle(
    Color=<rgba 0.9759 1 0.8912 2>
    Sprite Rotation Mode=::NewEnumerator2
    Sprite Size Mode=::NewEnumerator1
    Uniform Sprite Size Max=150
    Uniform Sprite Size Min=100
    UsePositionOffset=false
    Write Material Random=true
   )
   &AddVelocity(
    Module.Use Falloff Distance=false
    Origin Offset=<0 0 2500>
    Origin Offset Coordinate Space=::World
    Override Seed=false
    Velocity Mode=::NewEnumerator1
    Velocity Speed=7000
   )
   &SubUVAnimation(
    Evaluation Type=::NewEnumerator0
    "Sprite Renderer "=DataInterface() {propertyValues {SpriteRenderer=&self:GPU_Coll_Smoke_4.NiagaraSpriteRendererProperties_0}}
    SubUV Animation Mode:ENiagara_SubUVLookupModeV2=::NewEnumerator2
    UseEndFrame=false
    UseStartFrame=false
   ) {enabled=false}
   &ShapeLocation(
    MaskSpawnGroup=false
    Offset=<0 0 60>
    Offset Mode=::NewEnumerator0
    Override Seed=false
    Ring Radius=20
    Shape Primitive=::NewEnumerator4
    UseOwnerScale=true
   )
  ]
  particleUpdateScript [
   &ParticleState()
   &CameraOffset(Camera Offset Amount=100)
   &SpriteRotationRate(Rotation Rate=&Multiply_Float(
     A=&RandomRangeFloat(Minimum=-60, Maximum=60)
     B=&FloatFromCurve(
      FloatCurve=DataInterface() {propertyValues {Curve {keys [
          {interpMode=RCIM_Cubic, tangentMode=RCTM_User, time=0.1, value=0}
          {arriveTangent=2.1084, interpMode=RCIM_Cubic, leaveTangent=2.1084, tangentMode=RCTM_User, time=0.7, value=1}
         ]}}}
      Scale Curve=0.01
     )
    ))
   &ScaleSpriteSize(
    Uniform Curve Scale=&MakeCustomFloatFromBool(
     Bool=&SetBoolByFloatComparison(A=0.0001, B=HlslExpression() {hlslExpression="abs(Particles.Velocity.x + Particles.Velocity.y + Particles.Velocity.z)"})
     True Float=1.5
     False Float=0.01
    )
    Uniform Curve Sprite Scale=DataInterface() {propertyValues {Curve {keys [
        {arriveTangent=1.0284, interpMode=RCIM_Cubic, leaveTangent=1.0284, tangentMode=RCTM_User, time=0.01, value=0.1}
        {arriveTangent=1.0246, interpMode=RCIM_Cubic, leaveTangent=1.0246, tangentMode=RCTM_User, time=1, value=1.25}
       ]}}}
   )
   &Collision(
    Age Colliding Particles=false
    Collision Enabled=&SetBoolByFloatComparison(
     A=1
     B=&FloatFromCurve(FloatCurve=DataInterface() {propertyValues {Curve {keys [<at 0.1 0.95>,<at 0.5 2>]}}}, Scale Curve=0.5)
    )
    Enable Rest State=false
    Friction=5
    Friction During a Bounce=5
    GPU Collision Type=::NewEnumerator2
    Module.EnableMaxCollisionCount=false
    Module.Randomize Collision Normal Vector=false
    Particle Radius Scale=0
    Randomize Collision Normal=0
    Restitution=0
    Simple Friction=false
   )
   &Drag(Drag=20, Rotational Drag=-5, Use Linear Drag=true, Use Rotational Drag=false)
   &SolveForcesAndVelocity(Clamp Velocity=false, Limit Acceleration=false)
   &DynamicMaterialParameters(
    Index 0 Param 1=&RandomRangeFloat(Minimum=0, Maximum=1)
    Index 1 Param 1=&FloatFromCurve(FloatCurve=DataInterface() {propertyValues {Curve {keys [
         {arriveTangent=1.117, interpMode=RCIM_Cubic, leaveTangent=1.117, tangentMode=RCTM_User, time=0, value=0.1}
         {arriveTangent=0.0224, interpMode=RCIM_Cubic, leaveTangent=0.0224, tangentMode=RCTM_User, time=1, value=1}
        ]}}})
    Index 1 Param 3=&FloatFromCurve(
     FloatCurve=DataInterface() {propertyValues {Curve {keys [
         {arriveTangent=1.117, interpMode=RCIM_Cubic, leaveTangent=1.117, tangentMode=RCTM_User, time=0, value=0}
         {arriveTangent=0.0224, interpMode=RCIM_Cubic, leaveTangent=0.0224, tangentMode=RCTM_User, time=1, value=1}
        ]}}}
     Scale Curve=&RandomRangeFloat(Minimum=0.1, Maximum=0.2)
    )
    Module.Param0WriteEnabled=true
    Module.Param0WriteEnabled1=true
    Module.Param1WriteEnabled=false
    Module.Param1WriteEnabled1=true
    Module.Param2WriteEnabled=false
    Module.Param2WriteEnabled1=true
    Module.Param3WriteEnabled=false
    Module.Param3WriteEnabled1=true
    Write Parameter Index 1=true
   )
   &ScaleColor(
    Scale Alpha=&MakeCustomFloatFromBool(
     Bool=&SetBoolByFloatComparison(A=0.001, B=HlslExpression() {hlslExpression="abs(Particles.Velocity.x + Particles.Velocity.y + Particles.Velocity.z)"})
     True Float=&FloatFromCurve(
      FloatCurve=DataInterface() {propertyValues {Curve {keys [
          {tangentMode=RCTM_User, time=0, value=0}
          {interpMode=RCIM_Cubic, tangentMode=RCTM_User, time=0.1, value=0.15}
          {arriveTangent=-0.0405, interpMode=RCIM_Cubic, leaveTangent=-0.0405, tangentMode=RCTM_User, time=0.3, value=1}
          {arriveTangent=-0.3723, interpMode=RCIM_Cubic, leaveTangent=-0.3723, tangentMode=RCTM_User, time=1, value=0.05}
         ]}}}
      Scale Curve=&FNTimeOfDayScalar(Night Scale=0.6, Morning Scale=0.7, Evening Scale=0.7)
     )
     False Float=0
    )
    ScaleA=true
    ScaleRGB=true
   )
  ]
  rendererClasses [&NiagaraSpriteRendererProperties]
  renderers [{rendererClass=&NiagaraSpriteRendererProperties}]
 }
 SmokeLoopIdle {
  emitterUpdateScript [
   &EmitterState(
    Inactive Response=::NewEnumerator1
    Life Cycle Mode=::NewEnumerator1
    Loop Behavior=::NewEnumerator1
    Loop Delay=0.5
    Loop Duration Mode=::NewEnumerator1
    UseLoopDelay=false
   )
   &SpawnBurst_Instantaneous(Use Spawn Probability=false, UseLoopCountLimit=false)
   &SpawnRate(SpawnRate=20, Use Spawn Probability=false)
   &SpawnRate(SpawnGroup=1, SpawnRate=20, Use Spawn Probability=false) {module=SpawnRate001}
   /SetVariables(
    Emitter.CameraOfsCore=10
    Emitter.CameraOfs_Outline=-10
    Emitter.Color:LinearColor=&Multiply_LinearColorByFloat(
     Linear Color=&ColorFromCurve(
      ColorCurve=DataInterface() {propertyValues {
        AlphaCurve {keys [<at 0 0 RCIM_Cubic>,<at 0.5578 2 RCIM_Cubic>]}
        BlueCurve {keys [<at 0.4494 0 RCIM_Cubic>]}
        GreenCurve {keys [<at 0.4494 0 RCIM_Cubic>]}
        RedCurve {keys [<at 0.4494 0 RCIM_Cubic>]}
       }}
      CurveIndex=$Emitter.Age
     )
     Float=$System.CurveDestroyed
    )
    Emitter.Color_Core:LinearColor=&Multiply_LinearColorByFloat(Linear Color=<rgba 1 0.0208 0.3146 1>, Float=$System.CurveDestroyed)
    Emitter.Color_Outline:LinearColor=&Multiply_LinearColorByFloat(
     Linear Color=&ColorFromCurve(
      ColorCurve=DataInterface() {propertyValues {
        AlphaCurve {keys [<at 0.2072 0>,<at 0.6601 1 RCIM_Cubic>]}
        BlueCurve {keys [<at 0 0.614 RCIM_Cubic>]}
        GreenCurve {keys [<at 0 0.02 RCIM_Cubic>]}
        RedCurve {keys [<at 0 2 RCIM_Cubic>]}
       }}
      CurveIndex=$Emitter.Age
     )
     Float=$System.CurveDestroyed
    )
    Emitter.CoreDMP3:Vector4f=&MakeVector4(X=0.1, Y=0, Z=0, W=0)
   ) {isSetParameters=true}
   /SetVariables(
    Emitter.DynamicParam0_inky:Vector4f=&MakeVector4(
     X=&MakeFloatFromVector4(
      Vector 4=&MakeVector4(
       X=&FloatFromCurve(
        FloatCurve=DataInterface() {propertyValues {Curve {keys [<at 0 1 RCIM_Cubic>, {tangentMode=RCTM_User, time=1, value=0}]}}}
        CurveIndex=$Emitter.Age
        Scale Curve=0.5
       )
       Y=&RandomRangeFloat(Minimum=0, Maximum=1)
       Z=&FloatFromCurve(
        FloatCurve=DataInterface() {propertyValues {Curve {keys [{tangentMode=RCTM_User, time=0, value=0}, {tangentMode=RCTM_User, time=1, value=1}]}}}
        CurveIndex=$Emitter.Age
       )
       W=&FloatFromCurve(
        FloatCurve=DataInterface() {propertyValues {Curve {keys [{tangentMode=RCTM_User, time=0, value=1}, {tangentMode=RCTM_User, time=1, value=0}]}}}
        CurveIndex=$Emitter.Age
       )
      )
      Channel=::NewEnumerator0
     )
     Y=&FloatFromCurve(
      FloatCurve=DataInterface() {propertyValues {Curve {keys [{tangentMode=RCTM_User, time=0, value=0}, {tangentMode=RCTM_User, time=1, value=1}]}}}
      CurveIndex=$Emitter.Age
      Scale Curve=0.3
     )
     Z=&FloatFromCurve(
      FloatCurve=DataInterface() {propertyValues {Curve {keys [{tangentMode=RCTM_User, time=0, value=0}, {tangentMode=RCTM_User, time=1, value=1}]}}}
      CurveIndex=$Emitter.Age
     )
     W=&FloatFromCurve(
      FloatCurve=DataInterface() {propertyValues {Curve {keys [
          {interpMode=RCIM_Cubic, tangentMode=RCTM_User, time=0, value=1}
          {arriveTangent=-1.1812, interpMode=RCIM_Cubic, leaveTangent=-1.1812, tangentMode=RCTM_User, time=0.4, value=0.05}
          {interpMode=RCIM_Cubic, tangentMode=RCTM_User, time=1, value=1}
         ]}}}
      CurveIndex=$Emitter.Age
     )
    )
    Emitter.LightColor:LinearColor=<rgba 10 0 3 1>
    Emitter.LightRadius:NiagaraFloat=&Multiply_Float(
     A=&Add_Float(A=&Sine(Normalized Angle=$Emitter.Age, Period=0.78, Scale=0.2, Bias=1), B=1)
     B=100
    )
    Emitter.Scale:Vector3f=&VectorFromCurve(
     VectorCurve=DataInterface() {propertyValues {
       XCurve {keys [
         {arriveTangent=0.7455, interpMode=RCIM_Cubic, leaveTangent=0.7455, tangentMode=RCTM_User, time=0.4, value=0}
         {arriveTangent=0.7455, interpMode=RCIM_Cubic, leaveTangent=0.7455, tangentMode=RCTM_User, time=0.6, value=1.3}
         {arriveTangent=-0.5585, interpMode=RCIM_Cubic, leaveTangent=-0.5585, tangentMode=RCTM_User, time=0.8, value=1}
        ]}
       YCurve {keys [
         {arriveTangent=0.8624, interpMode=RCIM_Cubic, leaveTangent=0.8624, tangentMode=RCTM_User, time=0.4, value=0}
         {arriveTangent=0.8624, interpMode=RCIM_Cubic, leaveTangent=0.8624, tangentMode=RCTM_User, time=0.6, value=1.3}
         {arriveTangent=-2.2311, interpMode=RCIM_Cubic, leaveTangent=-2.2311, tangentMode=RCTM_User, time=0.8, value=1}
        ]}
       ZCurve {keys [
         {interpMode=RCIM_Cubic, tangentMode=RCTM_User, time=0.3, value=0}
         {arriveTangent=-1.954, interpMode=RCIM_Cubic, leaveTangent=-1.954, tangentMode=RCTM_User, time=0.55, value=1.35}
         {arriveTangent=-3.8442, interpMode=RCIM_Cubic, leaveTangent=-3.8442, tangentMode=RCTM_User, time=0.7, value=1}
        ]}
      }}
     CurveIndex=$Emitter.Age
     Scale Curve=<0.45 0.45 0.4>
    )
    Emitter.Scale2:Vector3f=&VectorFromCurve(
     VectorCurve=DataInterface() {propertyValues {XCurve {keys [<at 0.3 0>,<at 0.6 1>]}, YCurve {keys [<at 0.3 0>,<at 0.6 1>]}, ZCurve {keys [<at 0.4 0>,<at 0.5 1>]}}}
     CurveIndex=$Emitter.Age
     Scale Curve=<0.38 0.38 0.4>
    )
    Emitter.ScaleCore:Vector3f=&VectorFromCurve(
     VectorCurve=DataInterface() {propertyValues {XCurve {keys [<at 0.7 0>,<at 1.5 1>]}, YCurve {keys [<at 0.7 0>,<at 1.5 1>]}, ZCurve {keys [<at 0.7 0.8>,<at 1.5 1>]}}}
     CurveIndex=$Emitter.Age
     Scale Curve=<0.05 0.05 0.04>
    )
    Emitter.SpriteSize:Vector2f=&Multiply_Vector2D_ByFloat(
     Float=&MakeFloatFromVector(
      Vector=&Multiply_Vector(
       A=&VectorFromFloat(Value=&FloatFromCurve(
         FloatCurve=DataInterface() {propertyValues {Curve {keys [
             {arriveTangent=-1.0151, interpMode=RCIM_Cubic, leaveTangent=-1.0151, tangentMode=RCTM_User, time=0, value=1}
             {arriveTangent=-2.8172, interpMode=RCIM_Cubic, leaveTangent=-2.8172, tangentMode=RCTM_User, time=0.45, value=0.2}
            ]}}}
         CurveIndex=$Emitter.Age
        ))
       B=<1.4 1.4 0.4>
      )
      Channel=::NewEnumerator0
     )
     Vector2D=<220 200>
    )
   ) {isSetParameters=true}
   &SpawnBurst_Instantaneous(Spawn Time=8.55, Use Spawn Probability=false, UseLoopCountLimit=false) {module=SpawnBurst_Instantaneous001}
  ]
  particleSpawnScript [
   /SetVariables(Particles.VisibilityTag=$Emitter.SpawnGroup) {enabled=false, isSetParameters=true}
   &InitializeParticle(
    Color=&MaskLinearColorBySpawnGroup(Masked Linear Color=<rgba 0 0 0 1>, Passthrough Linear Color=<rgba 2.5 0 0.75 1>)
    Lifetime=&RandomRangeFloat(Minimum=1.3, Maximum=1.6)
    Sprite Rotation Mode=::NewEnumerator2
    Sprite Size=&Vector2DFromFloat(Value=&RandomRangeFloat(Minimum=135, Maximum=160))
    Sprite Size Mode=::NewEnumerator4
    Sprite UV Mode=::NewEnumerator3
    UsePositionOffset=false
    Write Material Random=true
   )
   /Niagara/Modules/Spawn/Location/StaticMeshLocation(
    MaskSpawnGroup=false
    OffsetAlongNormal=false
    OverrideRandomSeed=false
    SamplePosOffset=false
    Static Mesh=DataInterface() {propertyValues {
      DefaultMesh=null
      InstanceIndex=-1
      LODIndex=0
      PreviewMesh=/DemonMaskVoid/Assets/VFX/Meshes/SM_masks_DarkCore
      SoftSourceActor=null
      SourceMode=Default
      bAllowSamplingFromStreamingLODs=false
      bCaptureTransformsPerFrame=true
      bUsePhysicsBodyVelocity=false
      FilteredSockets []
      LODIndexUserParameter {parameter {name=None, typeDefHandle {registeredTypeIndex=108}, varData []}}
      MeshParameterBinding {parameter {name=None, typeDefHandle {registeredTypeIndex=1}, varData []}}
      SectionFilter {allowedMaterialSlots []}
     }}
   )
   /SetVariables(Particles.Position=&Multiply_Vector(A=$Particles.StaticMeshLocation.SampledPosition, B=<0.5 1 1>)) {enabled=false, isSetParameters=true}
   &CameraOffset(Camera Offset Amount=-100)
   &DynamicMaterialParameters(
    Index 0 Param 1=&RandomRangeFloat(Minimum=0, Maximum=-6)
    Module.Index0Param0Write=true
    Module.Index0Param1Write=false
    Module.Index0Param2Write=false
    Module.Index0Param3Write=false
   )
   &AddVelocity(Velocity=&RandomRangeVector(Minimum=<0 0 30>, Maximum=<0 0 -30>))
  ]
  particleUpdateScript [
   &ParticleState()
   &ScaleSpriteSize(
    Scale Sprite Size Mode=::NewEnumerator0
    Uniform Scale Factor=&FloatFromCurve(
     FloatCurve=DataInterface() {propertyValues {Curve {keys [
         <at 0 0 RCIM_Cubic>
         {arriveTangent=1.4143, interpMode=RCIM_Cubic, leaveTangent=1.4143, tangentMode=RCTM_User, time=0.53, value=0.4}
         {arriveTangent=0.9808, interpMode=RCIM_Cubic, leaveTangent=0.9808, tangentMode=RCTM_User, time=0.75, value=1}
         {arriveTangent=-8.1396, interpMode=RCIM_Cubic, leaveTangent=-8.1396, tangentMode=RCTM_User, time=1, value=0}
        ]}}}
     Scale Curve=0.7
    )
   )
   &ScaleColor(
    Scale Alpha=&Multiply_Float(
     A=$System.CurveDestroyed
     B=&FloatFromCurve(
      FloatCurve=DataInterface() {propertyValues {Curve {keys [
          <at 0 0 RCIM_Cubic>
          {arriveTangent=2.2188, interpMode=RCIM_Cubic, leaveTangent=2.2188, time=0.5, value=1}
          {arriveTangent=-2, interpMode=RCIM_Cubic, leaveTangent=-2, time=0.8, value=1.775}
          {interpMode=RCIM_Cubic, leaveTangent=-26.9591, time=1, value=0}
         ]}}}
      Scale Curve=&Multiply_Float(
       A=1
       B=&FloatFromCurve(
        FloatCurve=DataInterface() {propertyValues {Curve {keys [{tangentMode=RCTM_User, time=0.8, value=0}, {tangentMode=RCTM_User, time=1.3, value=1}]}}}
        CurveIndex=$Emitter.Age
       )
      )
     )
    )
    Scale RGB=&VectorFromFloat(Value=&FloatFromCurve(FloatCurve=DataInterface() {propertyValues {Curve {keys [
          {interpMode=RCIM_Cubic, tangentMode=RCTM_User, time=0, value=0.2}
          {interpMode=RCIM_Cubic, tangentMode=RCTM_User, time=0.25, value=1}
          {interpMode=RCIM_Cubic, tangentMode=RCTM_User, time=1, value=0}
         ]}}}))
    ScaleA=true
    ScaleRGB=true
   )
   /Niagara/Modules/Update/SubUV/SubUVAnimation(Number Of Frames=32)
   &DynamicMaterialParameters(
    Index 0 Param 1=&FloatFromCurve(
     FloatCurve=DataInterface() {propertyValues {Curve {keys [
         {arriveTangent=2.8718, interpMode=RCIM_Cubic, leaveTangent=2.8718, tangentMode=RCTM_User, time=0, value=0}
         {arriveTangent=-0.3942, interpMode=RCIM_Cubic, leaveTangent=-0.3942, tangentMode=RCTM_User, time=1, value=1}
        ]}}}
     Scale Curve=0.1
    )
    Index 0 Param 2=50
    Module.Index0Param0Write=true
    Module.Index0Param1Write=true
    Module.Index0Param2Write=false
    Module.Index0Param3Write=false
   ) {module=DynamicMaterialParameters001}
   &Drag(Use Linear Drag=true, Use Rotational Drag=true) {enabled=false}
   &PointAttractionForce(
    Attraction Radius=300
    Attraction Strength=&RandomRangeFloat(Minimum=50, Maximum=60)
    Falloff Exponent=0.1
    Kill Within Radius=false
    Module.Use Falloff=true
    UseAttractorOffset=false
   )
   &SolveForcesAndVelocity(Clamp Velocity=false, Limit Acceleration=false)
  ]
  rendererClasses [
   &NiagaraSpriteRendererProperties
   &NiagaraMeshRendererProperties
   &NiagaraLightRendererProperties
  ]
  renderers [
   {rendererClass=&NiagaraSpriteRendererProperties}
   {rendererClass=&NiagaraMeshRendererProperties, rendererIndex=1}
   {rendererClass=&NiagaraMeshRendererProperties, rendererIndex=2}
   {rendererClass=&NiagaraMeshRendererProperties, rendererIndex=3}
   {rendererClass=&NiagaraMeshRendererProperties, rendererIndex=4}
   {rendererClass=&NiagaraLightRendererProperties, rendererIndex=5}
   {rendererClass=&NiagaraSpriteRendererProperties, rendererIndex=6}
  ]
 }
}
summary {
 systemName=NS_Henchman_Portal_Core
 emitters [
  {
   emitterName=SmokeLoopIdle
   rendererClasses [
    &NiagaraSpriteRendererProperties
    &NiagaraMeshRendererProperties
    &NiagaraLightRendererProperties
   ]
  }
  {emitterName=GPU_Coll_Smoke, simTarget=GPUComputeSim, rendererClasses [&NiagaraSpriteRendererProperties]}
  {emitterName=CPU_Coll_Smoke001, rendererClasses [&NiagaraSpriteRendererProperties]}
  {emitterName=ChromaticRefraction, rendererClasses [&NiagaraSpriteRendererProperties]}
  {emitterName=Absorb_Sparkles, rendererClasses [&NiagaraSpriteRendererProperties]}
 ]
 userVariables [
  User.AimVec {type=Vector3f, description="", defaultValue {struct=Vector3f, value=<0 1 1>}}
  User.Destroyed {type=NiagaraBool}
  User.Offset Pos {type=NiagaraPosition, description="", defaultValue {}}
 ]
}
