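# Niagara system for the minor spire's pink-magenta energy beam, fired from the system position to User.Stop Location.
# Layered ribbons form the beam (Hermite spline positioning, per-spawn-group widths), alongside a core flare, chroma spheres, sparkles and head mesh.
# Impact lights and meshes sit at the stop point.
# System timelines, User.Duration and waveforms drive growth, fade and flicker, while time of day, beam length and camera distance modulate brightness.
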NiagaraSystem /SpireSharedAssets/Assets/VFX/NS_MinorSpire_Beam
dependencies {
 usedDataInterfaces [
  &NiagaraDataInterfaceCurve
  &NiagaraDataInterfaceVector2DCurve
  &NiagaraDataInterfaceColorCurve
  &NiagaraDataInterfaceCamera
 ]
 usedDynamicInputs [
  &VectorLength
  &Subtract_Position
  &RemapRange
 ]
 usedModules [
  &SystemState
  &InterpolateOverTime
  &self:SystemSpawnScript.SystemScriptSource.SystemScriptGraph.NiagaraNodeAssignment_5.SetVariables_#1
  &self:SystemSpawnScript.SystemScriptSource.SystemScriptGraph.NiagaraNodeAssignment_0.SetVariables_#2
  &self:SystemSpawnScript.SystemScriptSource.SystemScriptGraph.NiagaraNodeAssignment_1.SetVariables_#3
  &self:SystemSpawnScript.SystemScriptSource.SystemScriptGraph.NiagaraNodeAssignment_2.SetVariables_#4
  &self:SystemSpawnScript.SystemScriptSource.SystemScriptGraph.NiagaraNodeAssignment_3.SetVariables_#5
  &self:SystemSpawnScript.SystemScriptSource.SystemScriptGraph.NiagaraNodeAssignment_4.SetVariables_#6
  &self:SystemSpawnScript.SystemScriptSource.SystemScriptGraph.NiagaraNodeAssignment_6.SetVariables_#7
  &Timeline
 ]
}
emitters {
 Beams {
  emitterUpdateScript [
   &EmitterState()
   &SpawnBurst_Instantaneous(Spawn Count=20, Use Spawn Probability=false, UseLoopCountLimit=false)
   &SpawnBurst_Instantaneous(Spawn Count=20, Spawn Group=1, Use Spawn Probability=false, UseLoopCountLimit=false) {module=SpawnBurst_Instantaneous001}
   &SpawnBurst_Instantaneous(Spawn Count=20, Spawn Group=2, Use Spawn Probability=false, UseLoopCountLimit=false) {enabled=false, module=SpawnBurst_Instantaneous002}
   &SpawnBurst_Instantaneous(Spawn Count=20, Spawn Group=3, Use Spawn Probability=false, UseLoopCountLimit=false) {enabled=false, module=SpawnBurst_Instantaneous003}
   &SpawnBurst_Instantaneous(Spawn Count=20, Spawn Group=4, Use Spawn Probability=false, UseLoopCountLimit=false) {enabled=false, module=SpawnBurst_Instantaneous004}
   &SpawnBurst_Instantaneous(Spawn Count=20, Spawn Group=5, Use Spawn Probability=false, UseLoopCountLimit=false) {enabled=false, module=SpawnBurst_Instantaneous005}
   /SetVariables(Emitter.ScaleW:NiagaraFloat=&Waveform("Waveform [1]"=::NewEnumerator12, "[1] Frequency"=20, "[1] Amplitude Min/Max"=<0 1>)) {isSetParameters=true}
   /SetVariables(Emitter.BeamWidthScale:NiagaraFloat=&FloatFromCurve(
     FloatCurve=DataInterface() {propertyValues {Curve {keys [{interpMode=RCIM_Cubic, tangentMode=RCTM_Break, time=0, value=1}, <at 0.1 0>]}}}
     CurveIndex=&Add_Float(A=&Subtract_Float(A=$System.Age, B=$User.Duration), B=0.3)
    )) {isSetParameters=true}
  ]
  particleSpawnScript [
   /SetVariables(Particles.RibbonID=/Niagara/DynamicInputs/Meshes/MakeRibbonIDFromSpawnGroup()) {isSetParameters=true}
   &PartitionParticles(
    Number Of Partitions=1
    Particles Per Partition=20
    Partition Mode:ENiagara_PartitionMode=::NewEnumerator1
    Write Partition Index to Ribbon Link Order:ENiagara_AttributeSamplingApplyOutput=::NewEnumerator0
   )
   &InitializeParticle(
    Color=&MergeHelper_ColorSwitch(
     0=<rgba 0 0 0 0>
     1=<rgba 1 0.05 0.1292 0.5>
     2=<rgba 0 0 0 0>
     3=<rgba 2.5 0.025 0.2312 0.1>
     4=<rgba 0 0 0 0>
     5=<rgba 0 0 0 0>
     Index=&SpawnGroup()
     Num=::NewEnumerator4
    )
    Lifetime=$System.CurrentLoopDuration
    Ribbon Width=&MergeHelper_FloatSwitch(0=150, 1=220, 2=280, 3=700.3717, 4=200, 5=100, Index=&SpawnGroup(), Num=::NewEnumerator4)
    Ribbon Width Mode=::NewEnumerator1
    UsePositionOffset=false
    Write Material Random=true
   )
  ]
  particleUpdateScript [
   &ParticleState()
   /Game/Effects/Niagara/Modules/Curves/FN_HermiteSpline(
    Position Blend:NiagaraFloat=$Particles.Initial.PartitionParticles.Normalized Index In Partition
    Position End:NiagaraPosition=$User.Stop Location
    Position Start:NiagaraPosition=$Engine.Emitter.SimulationPosition
    Write Particles Position=true
   )
   /SetVariables(
    Particles.Beam Fade_Width:NiagaraFloat=&Multiply_Float(
     A=&Multiply_Float(A=$Particles.Initial.RibbonWidth, B=$Emitter.BeamWidthScale)
     B=&MaskFloatBySpawnGroup(Masked Float=1, Passthrough Float=0, Spawn Group Mask=2)
    )
    Particles.BeamDetail_Width:NiagaraFloat=&Multiply_Float(
     A=&Multiply_Float(A=$Particles.Initial.RibbonWidth, B=$Emitter.BeamWidthScale)
     B=&MaskFloatBySpawnGroup(
      Masked Float=/Game/Effects/Niagara/DynamicInputs/Math/FN_Fit(
       Clamp:NiagaraBool=false
       Value:NiagaraFloat=$Emitter.ScaleW
       SourceMin:NiagaraFloat=0
       SourceMax:NiagaraFloat=1
       DestMin:NiagaraFloat=0.7
       DestMax:NiagaraFloat=1
      )
      Passthrough Float=0
      Spawn Group Mask=4
     )
    )
    Particles.BeamGlow2_Width:NiagaraFloat=&Multiply_Float(
     A=&Multiply_Float(A=$Particles.Initial.RibbonWidth, B=$Emitter.BeamWidthScale)
     B=&MaskFloatBySpawnGroup(
      Masked Float=&FN_Fit(
       Clamp:NiagaraBool=false
       Value:NiagaraFloat=$Emitter.ScaleW
       SourceMin:NiagaraFloat=0
       SourceMax:NiagaraFloat=1
       DestMin:NiagaraFloat=0.25
       DestMax:NiagaraFloat=1
      )
      Passthrough Float=0
      Spawn Group Mask=3
     )
    )
    Particles.BeamGlow_Width:NiagaraFloat=&Multiply_Float(
     A=&Multiply_Float(A=$Particles.Initial.RibbonWidth, B=$Emitter.BeamWidthScale)
     B=&MaskFloatBySpawnGroup(Masked Float=1, Passthrough Float=0, Spawn Group Mask=1)
    )
    Particles.BeamParticle_Width:NiagaraFloat=&Multiply_Float(
     A=&Multiply_Float(A=$Particles.Initial.RibbonWidth, B=$Emitter.BeamWidthScale)
     B=&MaskFloatBySpawnGroup(Masked Float=1, Passthrough Float=0, Spawn Group Mask=5)
    )
    Particles.ProjectileBeam_Width:NiagaraFloat=&Multiply_Float(
     A=&Multiply_Float(A=$Particles.Initial.RibbonWidth, B=$Emitter.BeamWidthScale)
     B=&MaskFloatBySpawnGroup(Masked Float=1, Passthrough Float=0)
    )
   ) {isSetParameters=true}
   &ScaleColor(
    Scale Alpha=&MergeHelper_FloatSwitch(
     0=1
     1=1
     2=1
     3=&FloatFromCurve(
      FloatCurve=DataInterface() {propertyValues {Curve {keys [
          {interpMode=RCIM_Cubic, tangentMode=RCTM_Break, time=0, value=1}
          {arriveTangent=-3.624, interpMode=RCIM_Cubic, leaveTangent=-0.5922, tangentMode=RCTM_Break, time=0.7, value=0.3}
          {interpMode=RCIM_Cubic, tangentMode=RCTM_Break, time=1, value=0.4}
         ]}}}
      CurveIndex=$Particles.RibbonLinkOrder
     )
     4=1
     5=1
     Index=&SpawnGroup()
     Num=::NewEnumerator4
    )
    ScaleA=true
    ScaleRGB=true
   )
   &DynamicMaterialParameters(
    Index 0 Param 1=&MaskFloatBySpawnGroup(Masked Float=2, Passthrough Float=1)
    Index 2 Param 1=&Lerp_Float(
     A=-0.8775
     B=1
     Alpha=&Multiply_Float(
      A=&FloatFromCurve(FloatCurve=DataInterface() {propertyValues {Curve {keys [<at 0 0>,<at 0.1 1>]}}}, CurveIndex=$Engine.System.Age)
      B=1
     )
    )
    Index 3 Param 1=&FN_Fit(
     Clamp:NiagaraBool=false
     Value:NiagaraFloat=$Emitter.ScaleW
     SourceMin:NiagaraFloat=0
     SourceMax:NiagaraFloat=1
     DestMin:NiagaraFloat=0
     DestMax:NiagaraFloat=&FloatFromCurve(
      FloatCurve=DataInterface() {propertyValues {Curve {keys [{interpMode=RCIM_Cubic, tangentMode=RCTM_Break, time=0, value=0}, <at 0.5 1>]}}}
      CurveIndex=$Engine.System.Age
     )
    )
    Module.Index0Param0Write=true
    Module.Index0Param1Write=true
    Module.Index0Param2Write=false
    Module.Index0Param3Write=false
    Module.Index2Param0Write=true
    Module.Index2Param1Write=true
    Module.Index2Param2Write=true
    Module.Index2Param3Write=true
    Module.Index3Param0Write=true
    Module.Index3Param1Write=true
    Module.Index3Param2Write=true
    Module.Index3Param3Write=true
    Write Parameter Index 2=true
    Write Parameter Index 3=true
   )
   &ScaleColor(
    Scale Alpha=&FloatFromCurve(
     FloatCurve=DataInterface() {propertyValues {Curve {keys [{interpMode=RCIM_Cubic, leaveTangent=-3.7201, tangentMode=RCTM_Break, time=0, value=1}, <at 1 0>]}}}
     CurveIndex=$System.TimeForDis
    )
    ScaleA=true
    ScaleRGB=true
   ) {module=ScaleColor001}
  ]
  rendererClasses [&NiagaraRibbonRendererProperties]
  renderers [
   {rendererClass=&NiagaraRibbonRendererProperties}
   {rendererClass=&NiagaraRibbonRendererProperties, rendererIndex=1}
   {rendererClass=&NiagaraRibbonRendererProperties, rendererIndex=2}
   {rendererClass=&NiagaraRibbonRendererProperties, rendererIndex=3}
   {rendererClass=&NiagaraRibbonRendererProperties, rendererIndex=4}
   {rendererClass=&NiagaraRibbonRendererProperties, rendererIndex=5}
  ]
 }
 ChromaSphere_Mesh {
  emitterUpdateScript [
   &EmitterState()
   &SpawnBurst_Instantaneous(Use Spawn Probability=false, UseLoopCountLimit=true)
   &SpawnBurst_Instantaneous(Spawn Group=1, Use Spawn Probability=false, UseLoopCountLimit=true) {module=SpawnBurst_Instantaneous001}
  ]
  particleSpawnScript [
   /SetVariables(Particles.VisibilityTag=&SpawnGroup()) {isSetParameters=true}
   &InitializeParticle(
    Color=&MaskLinearColorBySpawnGroup(Masked Linear Color=<rgba 25 1.3515 3.3222 2>, Passthrough Linear Color=<rgba 1 0.1911 0.2585 1>)
    Lifetime Max=&MaskFloatBySpawnGroup(Masked Float=0.25, Passthrough Float=0.3)
    Lifetime Min:NiagaraFloat=&MaskFloatBySpawnGroup(Masked Float=0.1, Passthrough Float=0.2)
    Lifetime Mode=::NewEnumerator1
    Mesh Scale Mode=::NewEnumerator3
    Mesh Uniform Scale=&MaskFloatBySpawnGroup(Masked Float=3.5975, Passthrough Float=&RandomRangeFloat(Minimum=1.5, Maximum=2))
    UsePositionOffset=false
    Write Material Random=true
   )
   &InitialMeshOrientation(Use Rotation Vector=false)
   /SetVariables(Particles.MeshBOrient:Quat4f=&MakeQuatFromAxisAngle(Axis=<1 0 0>, AngleInDegrees=&RandomRangeFloat(Minimum=0, Maximum=360), NormalizeAxis=true)) {isSetParameters=true}
  ]
  particleUpdateScript [
   &ParticleState()
   &ScaleMeshSize(Scale Factor=&VectorFromFloat(Value=&MaskFloatBySpawnGroup(
      Masked Float=&FloatFromCurve(FloatCurve=DataInterface() {propertyValues {Curve {keys [
           {arriveTangent=0.7941, interpMode=RCIM_Cubic, leaveTangent=0.7941, leaveTangentWeight=1.1033, tangentMode=RCTM_User, time=0, value=0.5}
           <at 1 2>
          ]}}})
      Passthrough Float=&FloatFromCurve(FloatCurve=DataInterface() {propertyValues {Curve {keys [
           {arriveTangent=0.7941, interpMode=RCIM_Cubic, leaveTangent=0.7941, leaveTangentWeight=1.1033, tangentMode=RCTM_User, time=0, value=0.5}
           <at 1 1.5>
          ]}}})
     )))
   &DynamicMaterialParameters(
    Index 0 Param 1=&FloatFromCurve(FloatCurve=DataInterface() {propertyValues {Curve {keys [<at 0 0.07>,<at 1 0.25>]}}})
    Index 0 Param 2=&RandomRangeFloat(Minimum=0, Maximum=1)
    Index 0 Param 3=&RandomRangeFloat(Minimum=0, Maximum=1)
    Index 0 Param 4=&FloatFromCurve(FloatCurve=DataInterface() {propertyValues {Curve {keys [<at 0 1>,<at 1 3>]}}})
    Index 3 Param 1=&RandomRangeFloat(Minimum=3, Maximum=5)
    Module.Param0WriteEnabled=true
    Module.Param0WriteEnabled3=true
    Module.Param1WriteEnabled=true
    Module.Param1WriteEnabled3=true
    Module.Param2WriteEnabled=true
    Module.Param2WriteEnabled3=true
    Module.Param3WriteEnabled=true
    Module.Param3WriteEnabled3=true
    Write Parameter Index 3=true
   )
   &ScaleColor(
    Scale Alpha=&MaskFloatBySpawnGroup(
     Masked Float=&FloatFromCurve(FloatCurve=DataInterface() {propertyValues {Curve {keys [{interpMode=RCIM_Cubic, tangentMode=RCTM_User, time=0, value=1}, {interpMode=RCIM_Cubic, tangentMode=RCTM_User, time=1, value=0}]}}})
     Passthrough Float=&FloatFromCurve(FloatCurve=DataInterface() {propertyValues {Curve {keys [
          {interpMode=RCIM_Cubic, tangentMode=RCTM_User, time=0, value=0}
          {interpMode=RCIM_Cubic, tangentMode=RCTM_User, time=0.5, value=1}
          {tangentMode=RCTM_User, time=1, value=0}
         ]}}})
    )
    Scale RGB=&MaskVectorBySpawnGroup(
     Masked Vector=<1 1 1>
     Passthrough Vector=&VectorFromFloat(Value=&FloatFromCurve(
       FloatCurve=DataInterface() {propertyValues {Curve {keys [
           {arriveTangent=9.0725, interpMode=RCIM_Cubic, leaveTangent=9.0725, tangentMode=RCTM_User, time=0, value=0}
           {arriveTangent=-2.5249, interpMode=RCIM_Cubic, leaveTangent=-2.5249, tangentMode=RCTM_User, time=0.1408, value=0.9847}
           {arriveTangent=-0.0405, interpMode=RCIM_Cubic, leaveTangent=-0.0405, tangentMode=RCTM_User, time=1, value=0}
          ]}}}
       Scale Curve=5
      ))
    )
    ScaleA=true
    ScaleRGB=true
   )
   &ScaleSpriteSize(Uniform Curve Sprite Scale=DataInterface() {propertyValues {Curve {keys [
        <at 0 0 RCIM_Cubic>
        {arriveTangent=3.3072, interpMode=RCIM_Cubic, leaveTangent=3.3072, tangentMode=RCTM_User, time=0.011, value=0.85}
        {arriveTangent=0.6813, interpMode=RCIM_Cubic, leaveTangent=0.6813, tangentMode=RCTM_User, time=0.05, value=0.96}
        {arriveTangent=0.0129, interpMode=RCIM_Cubic, leaveTangent=0.0129, tangentMode=RCTM_User, time=0.171, value=0.98}
        <at 1 1 RCIM_Cubic>
       ]}}})
   &FN_Light_Attributes(
    Color=<rgba 2.4945 0.1507 5 1>
    Create Color Attribute=true
    Create Exponent Attribute=true
    Create Position Attribute=true
    Create Radius Attribute=true
    Exponent=2
    Position=&Add_Vector(A=Linked() {linkedVariable {name=Particles.Position, type=NiagaraPosition}}, B=<25 0 0>)
    Radius=100
    Scale Color=&FloatFromCurve(
     FloatCurve=DataInterface() {propertyValues {Curve {keys [
         {interpMode=RCIM_Cubic, tangentMode=RCTM_User, time=0, value=0}
         {interpMode=RCIM_Cubic, tangentMode=RCTM_User, time=0.5, value=1}
         {tangentMode=RCTM_User, time=1, value=0}
        ]}}}
     Scale Curve=&Waveform(
      Global Amplitude Scale=150
      "Waveform [1]"=::NewEnumerator2
      "[1] Frequency"=1
      "[1] Amplitude Min/Max"=<0.5 1.5>
      "[2] Phase"=$Emitter.Age
      "[3] Phase"=$Emitter.Age
     )
    )
   )
  ]
  rendererClasses [&NiagaraMeshRendererProperties, &NiagaraLightRendererProperties]
  renderers [
   {rendererClass=&NiagaraMeshRendererProperties, rendererIndex=0}
   {rendererClass=&NiagaraMeshRendererProperties, rendererIndex=1}
   {rendererClass=&NiagaraLightRendererProperties, rendererIndex=2}
  ]
 }
 CoreFlare {
  emitterSpawnScript [/SetVariables(Emitter.SizeCurve=DataInterface() {propertyValues {Curve {keys [
        <at 0.2 0 RCIM_Cubic>
        {arriveTangent=2.9237, interpMode=RCIM_Cubic, leaveTangent=1.25, tangentMode=RCTM_Break, time=0.8, value=0.95}
        {arriveTangent=-0.193, interpMode=RCIM_Cubic, leaveTangent=-0.193, tangentMode=RCTM_User, time=2, value=0}
       ]}}}) {isSetParameters=true}]
  emitterUpdateScript [
   &EmitterState()
   &SpawnBurst_Instantaneous(Use Spawn Probability=false, UseLoopCountLimit=true)
  ]
  particleSpawnScript [
   /SetVariables(Particles.LensColor:LinearColor=&MakeCustomLinearColorFromBool(Bool=$System.Outro, True Color=<rgba 1 0.0491 0.1284 1>, False Color=<rgba 1 0.141 0.2126 1>)) {isSetParameters=true}
   &InitializeParticle(
    Color=&MakeCustomLinearColorFromBool(Bool=$System.Outro, True Color=<rgba 0.1143 0 1 5>, False Color=<rgba 1 0.0139 0.0961 5>)
    Sprite Size Mode=::NewEnumerator3
    Uniform Sprite Size=1928.8291
    UsePositionOffset=true
    Write Material Random=true
   )
   /SetVariables(Particles.PLightColor:LinearColor=<rgba 1 0.0708 0.1482 5>) {isSetParameters=true}
  ]
  particleUpdateScript [
   &ParticleState()
   /SetVariables(Particles.FlareSizeUpdate:NiagaraFloat=&Multiply_Float(
     A=&RandomRangeFloat(
      Minimum=&Lerp_Float(A=1, B=0.8, Alpha=&FloatFromCurve(FloatCurve=$Emitter.SizeCurve))
      Maximum=1
      Evaluation Type=::NewEnumerator1
     )
     B=0.2
    )) {isSetParameters=true}
   /SetVariables(Particles.DarkRadialColor:LinearColor=&LerpLinearColors(StartColor=<rgba 0 0.0682 0.1615 2>, EndColor=<rgba 0.3333 0 0.0278 1>, LerpFactor=0)) {isSetParameters=true}
   &ScaleSpriteSize(Scale Factor=&Vector2DFromFloat(Value=&Multiply_Float(
      A=$Particles.FlareSizeUpdate
      B=&FN_LOD_Distance_Remap(Start Distance=500, End Distance=2500, Minimum Output=2.5, Maximum Output=5, Clamp Output=true)
     )))
   /SetVariables(Particles.DarkRadialSize:Vector2f=&Multiply_Vector2(
     A=<1500 700>
     B=&Vector2DFromFloat(Value=&Multiply_Float(A=$Particles.FlareSizeUpdate, B=1))
    )) {isSetParameters=true}
   /SetVariables(Particles.LensSizeUpdate:Vector2f=&Multiply_Vector2(
     A=<1500 250>
     B=&Vector2DFromFloat(Value=&Multiply_Float(A=$Particles.FlareSizeUpdate, B=5))
    )) {isSetParameters=true}
   /SetVariables(Particles.LightRadiusUpdate:NiagaraFloat=&Multiply_Float(A=150, B=1)) {isSetParameters=true}
   &CameraOffset(Camera Offset Amount=&MakeCustomFloatFromBool(Bool=$System.Outro, True Float=500, False Float=-300))
   /SetVariables(
    Particles.FlatFlareColor:LinearColor=&MultiplyLinearColor(
     Linear Color=<rgba 0.75 0.0197 0.0806 0.25>
     Scale Factor=&MakeFloat4FromVectorAndFloat(
      "Vector (RGB)"=&Multiply_VectorByFloat(
       Vector=&VectorFromFloat(Value=&Waveform("Waveform [1]"=::NewEnumerator12, "[1] Frequency"=10, "[1] Amplitude Min/Max"=<0.8 1.1>, "[1] Scale"=0.9))
       Float=1
      )
      "Float (Alpha)"=1
     )
    )
    Particles.FlatFlareDyn:Vector4f=&MakeVector4(
     X=0
     Y=&Waveform("Waveform [1]"=::NewEnumerator12, "[1] Frequency"=20, "[1] Amplitude Min/Max"=<0 0.2>)
     Z=0
     W=0
    )
    Particles.FlatFlareSize:Vector2f=&Multiply_Vector2D_ByFloat(
     Float=0.2
     Vector2D=&Multiply_Vector2(
      A=&Multiply_Vector2D(A=<-3200 440.6686>, B=<1 1>)
      B=&Vector2DFromFloat(Value=&Multiply_Float(A=1.1531, B=5))
     )
    )
   ) {isSetParameters=true}
   &ScaleColor(ScaleA=true, ScaleRGB=true)
   /SetVariables(Particles.LensColor:LinearColor=&MultiplyLinearColor(
     Linear Color=$Particles.Initial.LensColor
     Scale Factor=&MakeFloat4FromVectorAndFloat(
      "Vector (RGB)"=<1 1 1>
      "Float (Alpha)"=&FloatFromCurve(FloatCurve=DataInterface() {propertyValues {Curve {keys [
           {interpMode=RCIM_Cubic, tangentMode=RCTM_Break, time=0, value=1}
           {arriveTangent=-0.6275, interpMode=RCIM_Cubic, tangentMode=RCTM_Break, time=1, value=0}
          ]}}})
     )
    )) {isSetParameters=true}
  ]
  rendererClasses [&NiagaraSpriteRendererProperties]
  renderers [
   {rendererClass=&NiagaraSpriteRendererProperties}
   {rendererClass=&NiagaraSpriteRendererProperties, rendererIndex=1}
   {rendererClass=&NiagaraSpriteRendererProperties, rendererIndex=2}
  ]
 }
 EnergyBeam {
  emitterUpdateScript [
   &EmitterState()
   &SpawnBurst_Instantaneous(Spawn Count=6, Use Spawn Probability=false, UseLoopCountLimit=false)
   /Niagara/Modules/Beams/BeamEmitterSetup(
    Absolute Beam End=true
    Beam End:NiagaraPosition=/Niagara/DynamicInputs/Helpers/SimulationPosition()
    Beam Start:NiagaraPosition=$User.Stop Location
   )
  ]
  particleSpawnScript [
   &InitializeParticle(
    Lifetime=$System.CurrentLoopDuration
    Mesh Scale Mode=::NewEnumerator3
    Mesh Uniform Scale=50
    Position Mode=::NewEnumerator0
    Write Material Random=true
   )
   /Niagara/Modules/Beams/SpawnBeam()
   /Niagara/Modules/Beams/BeamWidth(Beam Width:NiagaraFloat=&Multiply_Float(
     A=&Lerp_Float(
      A=&Multiply_Float(
       A=20
       B=&Lerp_Float(A=1, B=&MakeFloatFromVector(Vector=$Engine.Owner.Scale, Channel=::NewEnumerator0), Alpha=1)
      )
      B=&Multiply_Float(
       A=100
       B=&Lerp_Float(A=1, B=&MakeFloatFromVector(Vector=$Engine.Owner.Scale, Channel=::NewEnumerator0), Alpha=0.3)
      )
      Alpha=&FloatFromCurve(
       FloatCurve=DataInterface() {propertyValues {Curve {keys [{tangentMode=RCTM_User, time=0, value=1}, {tangentMode=RCTM_User, time=1, value=0}]}}}
       CurveIndex=$Particles.RibbonLinkOrder
      )
     )
     B=1
    ))
   /SetVariables(
    Particles.BeamWidthBlack:NiagaraFloat=&Multiply_Float(A=$Particles.RibbonWidth, B=1.32)
    Particles.BeamWidthBlack02:NiagaraFloat=&Multiply_Float(A=$Particles.RibbonWidth, B=1.05)
    Particles.BlackEnergyColor:LinearColor=$Particles.Color
   ) {isSetParameters=true}
   /SetVariables(Particles.BeamWidthLightning:NiagaraFloat=&Multiply_Float(A=$Particles.RibbonWidth, B=2.8)) {isSetParameters=true}
   /SetVariables(
    Particles.LightningBeamAxis01:Vector3f=<1 0 1>
    Particles.LightningBeamAxis02:Vector3f=<1 0 -1>
    Particles.LightningBeamAxis03:Vector3f=<1 0 0>
   ) {isSetParameters=true}
   &InitialMeshOrientation(Mesh Orientation Mode=::NewEnumerator2, Orientation Vector=$System.BeamDirection, Use Rotation Vector=false)
  ]
  particleUpdateScript [
   &ParticleState()
   /Niagara/Modules/Beams/UpdateBeam()
   &DynamicMaterialParameters(
    Index 0 Param 1=&Lerp_Float(A=1, B=-0.075, Alpha=$System.BeamLengthScalar)
    Index 0 Param 3=0.9
    Index 0 Param 4=&FloatFromCurve(
     FloatCurve=DataInterface() {propertyValues {Curve {keys [{arriveTangent=4.1862, interpMode=RCIM_Cubic, leaveTangent=4.1862, tangentMode=RCTM_User, time=0, value=0}, <at 0.5 1 RCIM_Cubic>]}}}
     CurveIndex=$System.Timeline.Time
    )
    Index 1 Param 1=&MakeFloatFromVector(Vector=$Emitter.BeamDirection, Channel=::NewEnumerator0)
    Index 1 Param 2=&MakeFloatFromVector(Vector=$Emitter.BeamDirection, Channel=::NewEnumerator1)
    Index 1 Param 3=&MakeFloatFromVector(Vector=$Emitter.BeamDirection, Channel=::NewEnumerator2)
    Module.Index0Param0Write=true
    Module.Index0Param1Write=true
    Module.Index0Param2Write=true
    Module.Index0Param3Write=true
    Module.Index1Param0Write=true
    Module.Index1Param1Write=true
    Module.Index1Param2Write=true
    Module.Index1Param3Write=true
    Write Parameter Index 1=true
   )
   &SolveForcesAndVelocity(Clamp Velocity=false, Limit Acceleration=false)
   &ScaleColor(
    Scale Alpha=&Multiply_Float(
     A=&FNTimeOfDayScalar(Night Scale=0.15, Morning Scale=0.7, Evening Scale=0.7)
     B=&Lerp_Float(A=0.6, B=1, Alpha=$System.ImpactLODRemap)
    )
    Scale RGB=&VectorFromFloat(Value=&Lerp_Float(
      A=1
      B=10
      Alpha=&RandomRangeFloat(
       Minimum=&FNTimeOfDayScalar(Night Scale=1, Morning Scale=0.6, Day Scale=0.6, Evening Scale=0.6)
       Maximum=1
       Evaluation Type=::NewEnumerator1
      )
     ))
    ScaleA=true
    ScaleRGB=true
   )
   &ScaleColor(
    Scale Alpha=&FloatFromCurve(FloatCurve=DataInterface() {propertyValues {Curve {keys [<at 0 1 RCIM_Cubic>,<at 1 0 RCIM_Cubic>]}}})
    Scale RGB=&VectorFromFloat(Value=&Multiply_Float(A=1, B=&Lerp_Float(A=1, B=20, Alpha=$System.BeamShootoutFlash)))
    ScaleA=true
    ScaleRGB=true
   ) {module=ScaleColor001}
   /SetVariables(Particles.BlackEnergyColor:LinearColor=&MakeLinearColor(
     R:NiagaraFloat=1
     G:NiagaraFloat=1
     B:NiagaraFloat=1
     A:NiagaraFloat=&Multiply_Float(
      A=&FNTimeOfDayScalar(Night Scale=0.5, Morning Scale=0.7, Evening Scale=0.7)
      B=&Lerp_Float(A=1.5, B=2, Alpha=$System.ImpactLODRemap)
     )
    )) {isSetParameters=true}
   /Niagara/Modules/Beams/ScaleBeamWidth(Beam Width Scale:NiagaraFloat=&Multiply_Float(A=&Multiply_Float(A=$User.Size, B=$System.Timeline001.InverseNormalizedTime), B=4))
  ]
  rendererClasses [&NiagaraRibbonRendererProperties]
  renderers [
   {rendererClass=&NiagaraRibbonRendererProperties}
   {rendererClass=&NiagaraRibbonRendererProperties, rendererIndex=1}
   {rendererClass=&NiagaraRibbonRendererProperties, rendererIndex=2}
   {rendererClass=&NiagaraRibbonRendererProperties, rendererIndex=3}
   {rendererClass=&NiagaraRibbonRendererProperties, rendererIndex=4}
   {rendererClass=&NiagaraRibbonRendererProperties, rendererIndex=5}
  ]
 }
 Explode_Sparkles001 {
  simTarget=GPUComputeSim
  emitterUpdateScript [
   &EmitterState()
   &SpawnRate(Spawn Probability=$System.Timeline001.InverseNormalizedTime, SpawnRate=200, Use Spawn Probability=true)
  ]
  particleSpawnScript [
   &InitializeParticle(
    Color=$User.PrimaryColor
    Lifetime=&RandomRangeFloat(Minimum=0.125, Maximum=0.75)
    Sprite Rotation Mode=::NewEnumerator0
    Sprite Size Max=<16 10>
    Sprite Size Min=<3.5 20>
    Sprite Size Mode=::NewEnumerator2
    UsePositionOffset=false
    Write Material Random=true
   )
   /SetVariables(Particles.RandomCore:NiagaraFloat=&FloatFromCurve(
     FloatCurve=DataInterface() {propertyValues {Curve {keys [
         {arriveTangent=0.0559, interpMode=RCIM_Cubic, leaveTangent=0.0559, tangentMode=RCTM_User, time=0, value=0.2}
         {arriveTangent=0.1895, interpMode=RCIM_Cubic, leaveTangent=0.1895, tangentMode=RCTM_User, time=0.9, value=0.25}
         {arriveTangent=1.1808, interpMode=RCIM_Cubic, leaveTangent=1.1808, tangentMode=RCTM_User, time=1, value=0.3}
        ]}}}
     CurveIndex=&RandomRangeFloat(Minimum=0, Maximum=1)
     Scale Curve=2
    )) {isSetParameters=true}
   &ShapeLocation(MaskSpawnGroup=false, Override Seed=false, Sphere Radius=50, UseOwnerScale=true)
   &AddVelocity(
    Cone Angle=20
    Cone Axis=&NormalizeVector(
     Vector To Normalize=&SubtractVectorFromPosition(Position=$System.Stop Location Local, Vector=<0 0 0>)
     Safe Normalize=true
     Safe Fallback Threshold=0
     Safe Fallback Vector=<0 0 0.001>
    )
    Module.UseConeFalloff=false
    Module.UseVelDistribution=false
    Override Seed=false
    Rotation Mode=::NewEnumerator5
    Velocity Mode=::NewEnumerator2
    Velocity Speed=&RandomRangeFloat(Minimum=1000, Maximum=5000)
   )
  ]
  particleUpdateScript [
   &ParticleState()
   /SetVariables(Particles.Glow03_Color=&MakeLinearColorFromVectorAndFloat(
     "Vector (RGB)"=&MakeVectorFromLinearColorRGB(Color=<rgba 1 0.027 0.6368 5>)
     "Float (Alpha)"=&Multiply_Float(A=0.2, B=0.1)
    )) {isSetParameters=true}
   /SetVariables(Particles.Glow03_Size=&Multiply_Vector2D_ByFloat(
     Float=&FloatFromCurve(
      FloatCurve=DataInterface() {propertyValues {Curve {keys [
          <at 0 0 RCIM_Cubic>
          <at 0.2 1.3 RCIM_Cubic>
          {arriveTangent=-0.1423, interpMode=RCIM_Cubic, leaveTangent=-0.1423, tangentMode=RCTM_User, time=1, value=0}
         ]}}}
      Scale Curve=&Multiply_Float(A=125, B=5)
     )
     Vector2D=<1 1>
    )) {isSetParameters=true}
   &CurlNoiseForce(
    Noise Frequency=&RandomRangeFloat(Minimum=0.1, Maximum=0.5)
    Noise Strength=&RandomRangeFloat(Minimum=1500, Maximum=10000)
    Pan Noise=true
    Pan Noise Field=<1 0 0>
   ) {module=CurlNoiseForce001}
   &ScaleSpriteSize(Scale Factor=&MakeVector2D(
     X=&FloatFromCurve(
      FloatCurve=DataInterface() {propertyValues {Curve {keys [
          <at 0 0 RCIM_Cubic>
          <at 0.2 1.2 RCIM_Cubic>
          {arriveTangent=-2.7132, interpMode=RCIM_Cubic, leaveTangent=-2.7132, tangentMode=RCTM_User, time=1, value=0}
         ]}}}
      Scale Curve=&Multiply_Float(A=&RandomRangeFloat(Minimum=1, Maximum=5), B=1)
     )
     Y=&FloatFromCurve(
      FloatCurve=DataInterface() {propertyValues {Curve {keys [
          <at 0 0 RCIM_Cubic>
          {arriveTangent=0.2042, interpMode=RCIM_Cubic, leaveTangent=0.2042, tangentMode=RCTM_User, time=0.2, value=1.2}
          {arriveTangent=0.1752, interpMode=RCIM_Cubic, leaveTangent=0.1752, tangentMode=RCTM_User, time=1, value=0}
         ]}}}
      Scale Curve=&Multiply_Float(A=&RandomRangeFloat(Minimum=1, Maximum=5), B=2)
     )
    ))
   &ScaleColor(
    Scale Alpha=&Multiply_Float(A=0.1, B=0.1)
    Scale RGB=&Multiply_VectorByFloat(
     Vector=&VectorFromFloat(Value=3)
     Float=&FNTimeOfDayScalar(Night Scale=1, Morning Scale=1, Evening Scale=1)
    )
    ScaleA=true
    ScaleRGB=true
   )
   &Drag(Drag=&RandomRangeFloat(Minimum=1, Maximum=2), Use Linear Drag=true, Use Rotational Drag=false)
   &SolveForcesAndVelocity(Clamp Velocity=false, Limit Acceleration=false)
  ]
  rendererClasses [&NiagaraSpriteRendererProperties]
  renderers [{rendererClass=&NiagaraSpriteRendererProperties}]
 }
 Head {
  emitterUpdateScript [
   &EmitterState()
   &SpawnBurst_Instantaneous(Use Spawn Probability=false, UseLoopCountLimit=true)
  ]
  particleSpawnScript [&InitializeParticle(
    Color=$User.SecondaryColor
    Lifetime=$System.CurrentLoopDuration
    Mesh Scale=&Multiply_VectorByFloat(Vector=<15 30 30>, Float=0.1)
    Mesh Scale Mode=::NewEnumerator4
    Position Offset=&TransformPosition(Position=$System.BeamEndWs, Source Space=::World, Destination Space=::Local)
    UsePositionOffset=true
    Write Material Random=true
   )]
  particleUpdateScript [
   &ParticleState()
   /SetVariables(Particles.Position=$System.Stop Location Local) {isSetParameters=true}
   &UpdateMeshOrientation(
    Facing Direction=$System.BeamDirection
    Facing Interpolation=false
    Orientation Method=::NewEnumerator1
    Override Quat=false
    TransformFacing=false
    TransformUp=false
   )
   &ScaleColor(
    Scale Alpha=&Multiply_Float(A=$System.Timeline002.Time, B=$System.Timeline001.InverseNormalizedTime)
    Scale RGB=&Multiply_VectorByFloat(
     Vector=&MakeVectorFromLinearColorRGB(Color=<rgba 2.8 0 0.2333 1>)
     Float=&Multiply_Float(
      A=&Waveform("Waveform [1]"=::NewEnumerator12, "[1] Frequency"=16, "[1] Amplitude Min/Max"=<1 3>, "[1] Scale"=4)
      B=1
     )
    )
    ScaleA=true
    ScaleRGB=true
   )
  ]
  rendererClasses [&NiagaraMeshRendererProperties]
  renderers [
   {rendererClass=&NiagaraMeshRendererProperties, rendererIndex=0}
   {rendererClass=&NiagaraMeshRendererProperties, rendererIndex=1}
  ]
 }
 ImpactLight {
  emitterSpawnScript [/SetVariables(Emitter.CameraOffsetSprite=50) {isSetParameters=true}]
  emitterUpdateScript [
   &EmitterState()
   &SpawnBurst_Instantaneous(Use Spawn Probability=true, UseLoopCountLimit=true)
   &SpawnBurst_Instantaneous(Use Spawn Probability=false, UseLoopCountLimit=false) {module=SpawnBurst_Instantaneous001}
  ]
  particleSpawnScript [
   &InitializeParticle(
    Lifetime=0.6
    Position=$User.Stop Location
    Position Mode=::NewEnumerator1
    Sprite Rotation Mode=::NewEnumerator2
    Sprite Size Mode=::NewEnumerator3
    Uniform Sprite Size=400
    UsePositionOffset=false
    Write Material Random=true
   )
   &DynamicMaterialParameters(Module.Index0Param0Write=false, Module.Index0Param1Write=false, Module.Index0Param2Write=false, Module.Index0Param3Write=false)
  ]
  particleUpdateScript [
   &ParticleState()
   &ScaleColor(
    Scale Alpha=&Multiply_Float(
     A=&MakeCustomFloatFromBool(
      Bool=&SetBoolByFloatComparison(A=$System.BeamLength, B=9500, Comparison Type=::NewEnumerator6)
      True Float=0.5
      False Float=0.1
     )
     B=&Multiply_Float(
      A=&FNTimeOfDayScalar(Night Scale=0.2, Morning Scale=0.5, Evening Scale=0.5)
      B=&Lerp_Float(A=0.15, B=1, Alpha=$System.ImpactLODRemap)
     )
    )
    Scale RGB=<1 0.05 0.3>
    ScaleA=true
    ScaleRGB=true
   ) {module=ScaleColor002}
   /SetVariables(Particles.FlareColor=&MultiplyLinearColor(
     Linear Color=<rgba 1 0.0233 0.1047 1>
     Scale Factor=&MakeVector4(
      X=1
      Y=1
      Z=1
      W=&Multiply_Float(
       A=&Multiply_Float(
        A=&MakeCustomFloatFromBool(
         Bool=&SetBoolByFloatComparison(A=$System.BeamLength, B=9500, Comparison Type=::NewEnumerator6)
         True Float=1
         False Float=0.05
        )
        B=&Lerp_Float(A=0.3, B=1, Alpha=$System.ImpactLODRemap)
       )
       B=&FloatFromCurve(
        FloatCurve=DataInterface() {propertyValues {Curve {keys [<at 0 0.5>,<at 1 1>]}}}
        CurveIndex=&Divide_Float(
         A=&DistanceBetweenPositions(
          A=&ReturnCameraPosition(Coordinate Space:ENiagaraCoordinateSpace=::Simulation, Camera Query:NiagaraDataInterfaceCamera=DataInterface())
          B=$User.Stop Location
         )
         B=&Multiply_Float(A=300, B=&MakeFloatFromVector(Vector=$Engine.Owner.Scale, Channel=::NewEnumerator0))
        )
        Scale Curve=&FNTimeOfDayScalar(Night Scale=0.3, Morning Scale=0.7, Evening Scale=0.7)
       )
      )
     )
    )) {isSetParameters=true}
   &ApplyOwnerScaleToAttributes(Scale Initial Sprite Size=true, Sprite Size Scale Amount=0.2)
   &SpriteRotationRate(Rotation Rate=1000)
   &ScaleSpriteSize(
    Scale Sprite Size Mode=::NewEnumerator0
    Uniform Scale Factor=&Lerp_Float(
     A=0.5
     B=1
     Alpha=&RandomRangeFloat(
      Minimum=&FNTimeOfDayScalar(Night Scale=1, Morning Scale=0.7, Day Scale=0.6, Evening Scale=0.7)
      Maximum=1
      Evaluation Type=::NewEnumerator1
     )
    )
   ) {module=ScaleSpriteSize001}
   &ScaleSpriteSize(
    Scale Sprite Size Mode=::NewEnumerator0
    Uniform Scale Factor=&Multiply_Float(A=3, B=&Lerp_Float(A=0.5, B=1, Alpha=$System.ImpactLODRemap))
   )
   &FN_Light_Attributes(
    Color=&MultiplyLinearColor(
     Linear Color=<rgba 1 0.0738 0.4269 1>
     Scale Factor=&MakeVector4_FromFloat(Float=&FNTimeOfDayScalar(Night Scale=0.5, Morning Scale=1.5, Day Scale=2, Evening Scale=1.5))
    )
    Create Color Attribute=true
    Create Position Attribute=true
    Create Radius Attribute=true
    Enable Particle Light Enable=true
    Light Enabled=true
    Position=&AddVectorToPosition(
     Position=$Particles.Position
     Vector=&TransformVector(
      Vector=&Multiply_Vector(
       A=Linked() {linkedVariable {name=System.BeamDirection, type=NiagaraPosition}}
       B=&VectorFromFloat(Value=0)
      )
      Source Space=::World
      Destination Space=::Local
     )
    )
    Radius=&Multiply_Float(
     A=&Lerp_Float(A=400, B=800, Alpha=&RandomRangeFloat(Minimum=0.5, Maximum=1, Evaluation Type=::NewEnumerator1))
     B=&Lerp_Float(A=0.2, B=1, Alpha=$System.ImpactLODRemap)
    )
    Scale Color=20
   )
   /SetVariables(Particles.Position=$System.BeamEndLocation) {isSetParameters=true}
   &ScaleColor(
    Scale Alpha=&FloatFromCurve(
     FloatCurve=DataInterface() {propertyValues {Curve {keys [<at 0 0.3>,<at 1 1>]}}}
     CurveIndex=&Divide_Float(
      A=&DistanceBetweenPositions(
       A=&ReturnCameraPosition(Coordinate Space:ENiagaraCoordinateSpace=::Simulation, Camera Query:NiagaraDataInterfaceCamera=DataInterface())
       B=$User.Stop Location
      )
      B=&Multiply_Float(A=600, B=&MakeFloatFromVector(Vector=$Engine.Owner.Scale, Channel=::NewEnumerator0))
     )
    )
    ScaleA=true
    ScaleRGB=true
   )
  ]
  rendererClasses [&NiagaraSpriteRendererProperties, &NiagaraLightRendererProperties]
  renderers [
   {rendererClass=&NiagaraSpriteRendererProperties}
   {rendererClass=&NiagaraLightRendererProperties, rendererIndex=1}
   {rendererClass=&NiagaraSpriteRendererProperties, rendererIndex=2}
  ]
 }
 Impact_ChromaSphere_Mesh {
  emitterUpdateScript [
   &EmitterState()
   &SpawnBurst_Instantaneous(
    Spawn Probability=&MakeCustomFloatFromBool(
     Bool=&SetBoolByFloatComparison(A=$System.Timeline.NormalizedTime, B=0.9, Comparison Type=::NewEnumerator2)
     True Float=1
     False Float=0
    )
    Use Spawn Probability=true
    UseLoopCountLimit=false
   )
   &SpawnBurst_Instantaneous(
    Spawn Group=1
    Spawn Probability=&MakeCustomFloatFromBool(
     Bool=&SetBoolByFloatComparison(A=$System.Timeline.NormalizedTime, B=0.9, Comparison Type=::NewEnumerator2)
     True Float=1
     False Float=0
    )
    Use Spawn Probability=true
    UseLoopCountLimit=false
   ) {module=SpawnBurst_Instantaneous001}
  ]
  particleSpawnScript [
   /SetVariables(Particles.VisibilityTag=&SpawnGroup()) {isSetParameters=true}
   &InitializeParticle(
    Color=&MaskLinearColorBySpawnGroup(
     Masked Linear Color=<rgba 12.4726 0.7533 25 2>
     Passthrough Linear Color=&FN_Select_Color_from_Time_Of_Day(Morning=<rgba 0.8667 0 1 1>, Day=<rgba 0.4989 0.0301 1 1>, Evening=<rgba 0.4989 0.0301 1 1>, Night=<rgba 0.8667 0 1 1>)
    )
    Lifetime Max=0.25
    Lifetime Min=0.1
    Lifetime Mode=::NewEnumerator1
    Mesh Scale Mode=::NewEnumerator3
    Mesh Uniform Scale=&MaskFloatBySpawnGroup(Masked Float=4.25, Passthrough Float=&RandomRangeFloat(Minimum=3, Maximum=4))
    Position=$User.Stop Location
    Position Mode=::NewEnumerator1
    UsePositionOffset=true
    Write Material Random=true
   )
   &InitialMeshOrientation(Use Rotation Vector=false)
   /SetVariables(Particles.MeshB_Rotation:Quat4f=&MakeQuatFromAxisAngle(Axis=<1 0 0>, AngleInDegrees=&RandomRangeFloat(Minimum=0, Maximum=360), NormalizeAxis=true)) {isSetParameters=true}
  ]
  particleUpdateScript [
   &ParticleState()
   /SetVariables(Particles.Position=$User.Stop Location) {isSetParameters=true}
   &ScaleMeshSize(Scale Factor=&VectorFromFloat(Value=&MaskFloatBySpawnGroup(
      Masked Float=&FloatFromCurve(FloatCurve=DataInterface() {propertyValues {Curve {keys [
           {arriveTangent=0.7941, interpMode=RCIM_Cubic, leaveTangent=0.7941, leaveTangentWeight=1.1033, tangentMode=RCTM_User, time=0, value=0.5}
           <at 1 2>
          ]}}})
      Passthrough Float=&FloatFromCurve(FloatCurve=DataInterface() {propertyValues {Curve {keys [
           {arriveTangent=0.7941, interpMode=RCIM_Cubic, leaveTangent=0.7941, leaveTangentWeight=1.1033, tangentMode=RCTM_User, time=0, value=0.5}
           <at 1 1.5>
          ]}}})
     )))
   &DynamicMaterialParameters(
    Index 0 Param 1=&FloatFromCurve(FloatCurve=DataInterface() {propertyValues {Curve {keys [<at 0 0.07>,<at 1 0.25>]}}})
    Index 0 Param 2=&RandomRangeFloat(Minimum=0, Maximum=1)
    Index 0 Param 3=&RandomRangeFloat(Minimum=0, Maximum=1)
    Index 0 Param 4=&FloatFromCurve(FloatCurve=DataInterface() {propertyValues {Curve {keys [<at 0 1>,<at 1 3>]}}})
    Index 3 Param 1=&RandomRangeFloat(Minimum=3, Maximum=5)
    Module.Param0WriteEnabled=true
    Module.Param0WriteEnabled3=true
    Module.Param1WriteEnabled=true
    Module.Param1WriteEnabled3=true
    Module.Param2WriteEnabled=true
    Module.Param2WriteEnabled3=true
    Module.Param3WriteEnabled=true
    Module.Param3WriteEnabled3=true
    Write Parameter Index 3=true
   )
   &ScaleColor(
    Scale Alpha=&MaskFloatBySpawnGroup(
     Masked Float=&FloatFromCurve(FloatCurve=DataInterface() {propertyValues {Curve {keys [{interpMode=RCIM_Cubic, tangentMode=RCTM_User, time=0, value=1}, {interpMode=RCIM_Cubic, tangentMode=RCTM_User, time=1, value=0}]}}})
     Passthrough Float=&FloatFromCurve(FloatCurve=DataInterface() {propertyValues {Curve {keys [
          {interpMode=RCIM_Cubic, tangentMode=RCTM_User, time=0, value=0}
          {interpMode=RCIM_Cubic, tangentMode=RCTM_User, time=0.5, value=1}
          {tangentMode=RCTM_User, time=1, value=0}
         ]}}})
    )
    Scale RGB=&MaskVectorBySpawnGroup(
     Masked Vector=<1 1 1>
     Passthrough Vector=&VectorFromFloat(Value=&FloatFromCurve(
       FloatCurve=DataInterface() {propertyValues {Curve {keys [
           {arriveTangent=9.0725, interpMode=RCIM_Cubic, leaveTangent=9.0725, tangentMode=RCTM_User, time=0, value=0}
           {arriveTangent=-2.5249, interpMode=RCIM_Cubic, leaveTangent=-2.5249, tangentMode=RCTM_User, time=0.1408, value=0.9847}
           {arriveTangent=-0.0405, interpMode=RCIM_Cubic, leaveTangent=-0.0405, tangentMode=RCTM_User, time=1, value=0}
          ]}}}
       Scale Curve=5
      ))
    )
    ScaleA=true
    ScaleRGB=true
   )
   &FN_Light_Attributes(
    Color=<rgba 2.4945 0.1507 5 1>
    Create Color Attribute=true
    Create Exponent Attribute=true
    Create Position Attribute=true
    Create Radius Attribute=true
    Exponent=5
    Position=&Add_Vector(A=Linked() {linkedVariable {name=Particles.Position, type=NiagaraPosition}}, B=<0 0 0>)
    Radius=400
    Scale Color=&FloatFromCurve(
     FloatCurve=DataInterface() {propertyValues {Curve {keys [
         {interpMode=RCIM_Cubic, tangentMode=RCTM_User, time=0, value=0}
         {interpMode=RCIM_Cubic, tangentMode=RCTM_User, time=0.5, value=1}
         {tangentMode=RCTM_User, time=1, value=0}
        ]}}}
     Scale Curve=&Waveform(
      Global Amplitude Scale=150
      "Waveform [1]"=::NewEnumerator2
      "[1] Frequency"=1
      "[1] Amplitude Min/Max"=<0.5 2>
      "[2] Phase"=$Emitter.Age
      "[3] Phase"=$Emitter.Age
     )
    )
   )
  ]
  rendererClasses [&NiagaraMeshRendererProperties, &NiagaraLightRendererProperties]
  renderers [
   {rendererClass=&NiagaraMeshRendererProperties, rendererIndex=0}
   {rendererClass=&NiagaraMeshRendererProperties, rendererIndex=1}
   {rendererClass=&NiagaraLightRendererProperties, rendererIndex=2}
  ]
 }
 Impact_Inner_Flares {
  emitterUpdateScript [
   &EmitterState()
   &SpawnRate(Spawn Probability=$System.Timeline002.Time, SpawnRate=20, Use Spawn Probability=true)
  ]
  particleSpawnScript [
   &InitializeParticle(
    Color=$User.PrimaryColor
    Lifetime Max=0.175
    Lifetime Min=0.075
    Lifetime Mode=::NewEnumerator1
    Position=$User.Stop Location
    Position Mode=::NewEnumerator1
    Sprite Rotation Mode=::NewEnumerator2
    Sprite Size Mode=::NewEnumerator1
    Uniform Sprite Size Max=700
    Uniform Sprite Size Min=500
    UsePositionOffset=false
    Write Material Random=true
   )
   &CameraOffset(Camera Offset Amount=20)
  ]
  particleUpdateScript [
   &ParticleState()
   /SetVariables(Particles.Position=$User.Stop Location) {isSetParameters=true}
   &ScaleColor(
    Scale Alpha=&FloatFromCurve(
     FloatCurve=DataInterface() {propertyValues {Curve {keys [
         {interpMode=RCIM_Cubic, tangentMode=RCTM_User, time=0, value=0}
         {interpMode=RCIM_Cubic, tangentMode=RCTM_User, time=0.5, value=1}
         {tangentMode=RCTM_User, time=1, value=0}
        ]}}}
     Scale Curve=&Multiply_Float(A=$System.Timeline002.Time, B=$System.Timeline001.InverseNormalizedTime)
    )
    Scale RGB=<4 2 2>
    ScaleA=true
    ScaleRGB=true
   )
   /SetVariables(Particles.Soft Glow Color:LinearColor=&MakeLinearColorFromVectorAndFloat(
     "Vector (RGB)"=&MakeVectorFromLinearColorRGB(Color=$User.SecondaryColor)
     "Float (Alpha)"=&FloatFromCurve(
      FloatCurve=DataInterface() {propertyValues {Curve {keys [
          {interpMode=RCIM_Cubic, tangentMode=RCTM_User, time=0, value=0}
          {interpMode=RCIM_Cubic, tangentMode=RCTM_User, time=0.5, value=1}
          {tangentMode=RCTM_User, time=1, value=0}
         ]}}}
      Scale Curve=0.2
     )
    )) {isSetParameters=true, module=SetVariables_#8}
   /SetVariables(
    Particles.Animorphic Color:LinearColor=&MakeLinearColorFromVectorAndFloat("Vector (RGB)"=&MakeVectorFromLinearColorRGB(Color=<rgba 4.989 0.3013 10 1>), "Float (Alpha)"=0.05)
    Particles.Animorphic Size:Vector2f=&RandomRangeVector2D(Minimum=<400 6>, Maximum=<600 12>)
    Particles.Soft Glow Size:Vector2f=&Vector2DFromFloat(Value=&RandomRangeFloat(Minimum=120, Maximum=160))
   ) {isSetParameters=true, module=SetVariables_#9}
   /SetVariables(
    Particles.Corona Color:LinearColor=&MakeLinearColorFromVectorAndFloat(
     "Vector (RGB)"=&MakeVectorFromLinearColorRGB(Color=<rgba 1.5 0.1334 0.2679 1>)
     "Float (Alpha)"=&FloatFromCurve(
      FloatCurve=DataInterface() {propertyValues {Curve {keys [
          <at 0 0 RCIM_Cubic>
          {arriveTangent=2.5, interpMode=RCIM_Cubic, leaveTangent=2.5, time=0.165, value=0.65}
          {arriveTangent=0.1786, interpMode=RCIM_Cubic, leaveTangent=0.1786, time=0.4, value=1}
          {arriveTangent=-1.6667, interpMode=RCIM_Cubic, leaveTangent=-1.6667, time=0.725, value=0.75}
          <at 1 0 RCIM_Cubic>
         ]}}}
      Scale Curve=&Multiply_Float(A=0.5, B=&Multiply_Float(A=$System.Timeline002.Time, B=$System.Timeline001.InverseNormalizedTime))
     )
    )
    Particles.Corona Size:Vector2f=&Vector2DFromFloat(Value=&FloatFromCurve(
      FloatCurve=DataInterface() {propertyValues {Curve {keys [<at 0 1>,<at 0.5 0.7>,<at 1 1>]}}}
      Scale Curve=&RandomRangeFloat(Minimum=750, Maximum=950)
     ))
   ) {isSetParameters=true}
   &DynamicMaterialParameters(
    Index 3 Param 1=&FloatFromCurve(FloatCurve=DataInterface() {propertyValues {Curve {keys [<at 0 0.7 RCIM_Cubic>,<at 0.5 0.25 RCIM_Cubic>,<at 1 0.7 RCIM_Cubic>]}}})
    Index 3 Param 2=1
    Index 3 Param 3=&RandomRangeFloat(Minimum=0.25, Maximum=0.35)
    Index 3 Param 4=1
    Module.Param0WriteEnabled3=true
    Module.Param1WriteEnabled3=true
    Module.Param2WriteEnabled3=true
    Module.Param3WriteEnabled3=true
    Write Parameter Index 0=false
    Write Parameter Index 3=true
   )
   &FN_Sprite_Attributes(
    Color=&Multiply_LinearColorByFloat(
     Linear Color=&MultiplyLinearColor(Linear Color=$User.SecondaryColor, Scale Factor=&MakeVector4_FromLinearColor(LinearColor=<rgba 1 1 1 0.1>))
     Float=&Multiply_Float(A=$System.Timeline002.Time, B=$System.Timeline001.InverseNormalizedTime)
    )
    Create Color Attribute=true
    Create SpriteSize Attribute=true
    SpriteSize=&Vector2DFromFloat(Value=&FloatFromCurve(FloatCurve=DataInterface() {propertyValues {Curve {keys [<at 0 1>,<at 0.5 0.9>,<at 1 1>]}}}, Scale Curve=929.0022))
    Write to Module Namespace=true
   )
  ]
  rendererClasses [&NiagaraSpriteRendererProperties]
  renderers [
   {rendererClass=&NiagaraSpriteRendererProperties}
   {rendererClass=&NiagaraSpriteRendererProperties, rendererIndex=1}
   {rendererClass=&NiagaraSpriteRendererProperties, rendererIndex=2}
   {rendererClass=&NiagaraSpriteRendererProperties, rendererIndex=3}
   {rendererClass=&NiagaraSpriteRendererProperties, rendererIndex=4}
   {rendererClass=&NiagaraSpriteRendererProperties, rendererIndex=5}
  ]
 }
 Muzzle_Inner_Flares {
  emitterUpdateScript [
   &EmitterState()
   &SpawnRate(SpawnRate=40, Use Spawn Probability=false)
  ]
  particleSpawnScript [
   &InitializeParticle(
    Color=$User.PrimaryColor
    Lifetime Max=0.175
    Lifetime Min=0.075
    Lifetime Mode=::NewEnumerator1
    Sprite Rotation Mode=::NewEnumerator2
    Sprite Size Mode=::NewEnumerator1
    Uniform Sprite Size Max=488.912
    Uniform Sprite Size Min=418.452
    UsePositionOffset=false
    Write Material Random=true
   )
   &CameraOffset(Camera Offset Amount=50)
   /SetVariables(
    Particles.Animorphic Color:LinearColor=<rgba 10 0.3013 1.1095 0.05>
    Particles.Animorphic Size:Vector2f=&RandomRangeVector2D(Minimum=<400 6>, Maximum=<600 12>)
    Particles.Soft Glow Size:Vector2f=&Vector2DFromFloat(Value=&RandomRangeFloat(Minimum=80, Maximum=120))
   ) {isSetParameters=true, module=SetVariables_#9}
  ]
  particleUpdateScript [
   &ParticleState()
   &ScaleColor(
    Scale Alpha=&FloatFromCurve(FloatCurve=DataInterface() {propertyValues {Curve {keys [
         {interpMode=RCIM_Cubic, tangentMode=RCTM_User, time=0, value=0}
         {interpMode=RCIM_Cubic, tangentMode=RCTM_User, time=0.5, value=1}
         {tangentMode=RCTM_User, time=1, value=0}
        ]}}})
    ScaleA=true
    ScaleRGB=true
   )
   /SetVariables(Particles.Soft Glow Color:LinearColor=&MakeLinearColorFromVectorAndFloat(
     "Vector (RGB)"=&MakeVectorFromLinearColorRGB(Color=<rgba 10 0.3013 1.1095 1>)
     "Float (Alpha)"=&FloatFromCurve(
      FloatCurve=DataInterface() {propertyValues {Curve {keys [
          {interpMode=RCIM_Cubic, tangentMode=RCTM_User, time=0, value=0}
          {interpMode=RCIM_Cubic, tangentMode=RCTM_User, time=0.5, value=1}
          {tangentMode=RCTM_User, time=1, value=0}
         ]}}}
      Scale Curve=0.2
     )
    )) {isSetParameters=true, module=SetVariables_#8}
   /SetVariables(
    Particles.Corona Color:LinearColor=&MakeLinearColorFromVectorAndFloat(
     "Vector (RGB)"=&MakeVectorFromLinearColorRGB(Color=<rgba 1.5 0.1334 0.2473 1>)
     "Float (Alpha)"=&FloatFromCurve(
      FloatCurve=DataInterface() {propertyValues {Curve {keys [
          <at 0 0 RCIM_Cubic>
          {arriveTangent=2.5, interpMode=RCIM_Cubic, leaveTangent=2.5, time=0.165, value=0.65}
          {arriveTangent=0.1786, interpMode=RCIM_Cubic, leaveTangent=0.1786, time=0.4, value=1}
          {arriveTangent=-1.6667, interpMode=RCIM_Cubic, leaveTangent=-1.6667, time=0.725, value=0.75}
          <at 1 0 RCIM_Cubic>
         ]}}}
      Scale Curve=0.5
     )
    )
    Particles.Corona Size:Vector2f=&Vector2DFromFloat(Value=&FloatFromCurve(
      FloatCurve=DataInterface() {propertyValues {Curve {keys [<at 0 1>,<at 0.5 0.7>,<at 1 1>]}}}
      Scale Curve=&RandomRangeFloat(Minimum=625, Maximum=675)
     ))
   ) {isSetParameters=true}
   &DynamicMaterialParameters(
    Index 3 Param 1=&FloatFromCurve(FloatCurve=DataInterface() {propertyValues {Curve {keys [<at 0 0.7 RCIM_Cubic>,<at 0.5 0.25 RCIM_Cubic>,<at 1 0.7 RCIM_Cubic>]}}})
    Index 3 Param 2=1
    Index 3 Param 3=&RandomRangeFloat(Minimum=0.25, Maximum=0.35)
    Index 3 Param 4=1
    Module.Param0WriteEnabled3=true
    Module.Param1WriteEnabled3=true
    Module.Param2WriteEnabled3=true
    Module.Param3WriteEnabled3=true
    Write Parameter Index 0=false
    Write Parameter Index 3=true
   )
   &ScaleColor(Scale Alpha=$System.Timeline002.InverseNormalizedTime, ScaleA=true, ScaleRGB=true) {module=ScaleColor001}
  ]
  rendererClasses [&NiagaraSpriteRendererProperties]
  renderers [
   {rendererClass=&NiagaraSpriteRendererProperties}
   {rendererClass=&NiagaraSpriteRendererProperties, rendererIndex=1}
   {rendererClass=&NiagaraSpriteRendererProperties, rendererIndex=2}
   {rendererClass=&NiagaraSpriteRendererProperties, rendererIndex=3}
  ]
 }
 PP {
  emitterUpdateScript [
   &EmitterState()
   &SpawnBurst_Instantaneous(Use Spawn Probability=false, UseLoopCountLimit=true)
  ]
  particleSpawnScript [&InitializeParticle(
    Color Mode=::NewEnumerator0
    Lifetime=$System.CurrentLoopDuration
    Sprite Size Mode=::NewEnumerator3
    Uniform Sprite Size=200
    UsePositionOffset=false
    Write Material Random=true
   )]
  particleUpdateScript [
   /SetVariables(Particles.SpeedNormalizedB:NiagaraFloat=&FloatFromCurve(
     FloatCurve=DataInterface() {propertyValues {Curve {keys [<at 0 1 RCIM_Cubic>, {arriveTangent=-1, interpMode=RCIM_Cubic, leaveTangent=-1, time=0.6, value=0.7}, <at 1 0 RCIM_Cubic>]}}}
     CurveIndex=$System.NormalizedLoopAge
     Scale Curve=&Multiply_Float(
      A=1
      B=&FN_LOD_Distance_Remap(Start Distance=5500, End Distance=6500, Minimum Output=1, Maximum Output=0, Clamp Output=true)
     )
    )) {isSetParameters=true}
   &ParticleState()
  ]
  rendererClasses [&NiagaraComponentRendererProperties]
  renderers [{rendererClass=&NiagaraComponentRendererProperties, rendererIndex=0}]
 }
}
summary {
 systemName=NS_MinorSpire_Beam
 emitters [
  {emitterName=Muzzle_Inner_Flares, rendererClasses [&NiagaraSpriteRendererProperties]}
  {emitterName=Impact_Inner_Flares, rendererClasses [&NiagaraSpriteRendererProperties]}
  {
   emitterName=ChromaSphere_Mesh
   rendererClasses [&NiagaraMeshRendererProperties, &NiagaraLightRendererProperties]
  }
  {
   emitterName=Impact_ChromaSphere_Mesh
   rendererClasses [&NiagaraMeshRendererProperties, &NiagaraLightRendererProperties]
  }
  {emitterName=Explode_Sparkles001, simTarget=GPUComputeSim, rendererClasses [&NiagaraSpriteRendererProperties]}
  {emitterName=Head, rendererClasses [&NiagaraMeshRendererProperties]}
  {emitterName=CoreFlare, rendererClasses [&NiagaraSpriteRendererProperties]}
  {emitterName=PP, rendererClasses [&NiagaraComponentRendererProperties]}
  {emitterName=EnergyBeam, rendererClasses [&NiagaraRibbonRendererProperties]}
  {emitterName=ImpactLight, rendererClasses [&NiagaraSpriteRendererProperties, &NiagaraLightRendererProperties]}
  {emitterName=Beams, rendererClasses [&NiagaraRibbonRendererProperties]}
 ]
 userVariables [
  User.Duration {type=NiagaraFloat, defaultValue=1, description=""}
  User.InterpolateRateOfChange {type=NiagaraFloat, defaultValue=1, description=""}
  User.PrimaryColor {type=LinearColor, description="", defaultValue {struct=LinearColor, value=<rgba 5 0.2882 0.6808 1>}}
  User.SecondaryColor {type=LinearColor, description="", defaultValue {struct=LinearColor, value=<rgba 0.7101 0.1187 0.168 5>}}
  User.Size {type=NiagaraFloat, defaultValue=1.5, description=""}
  User.Stop Location {type=NiagaraPosition, description="", defaultValue {}}
  User.Stop Normal {type=Vector3f, description="", defaultValue {struct=Vector3f, value=<0 0 1>}}
 ]
}
