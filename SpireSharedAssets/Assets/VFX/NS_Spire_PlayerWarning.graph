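# Niagara system that warns the player of a spire attack with a mesh edge glow, red energy beam and telegraph decal.
# EnergyBeam stretches a two-particle ribbon between sampled skeletal mesh positions, with User.IsMinor choosing the start point.
# A system-level InterpolateOverTime value drives glow and beam alpha curves and decal fade, which a Timeline also modulates.
# User.Duration sets beam lifetime and decal loop length.
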NiagaraSystem /SpireSharedAssets/Assets/VFX/NS_Spire_PlayerWarning
dependencies {
 usedDataInterfaces [
  &NiagaraDataInterfaceCurve
  &NiagaraDataInterfaceArrayFloat
  &NiagaraDataInterfaceCollisionQuery
 ]
 usedDynamicInputs [&MakeCustomFloatFromBool]
 usedModules [
  &SystemState
  &Timeline
  &SampleSkeletalMesh
  &self:ScratchModule
  &InterpolateOverTime
 ]
}
emitters {
 EdgeGlow {
  emitterUpdateScript [
   &EmitterState()
   &SpawnBurst_Instantaneous(Use Spawn Probability=false, UseLoopCountLimit=false) {module=SpawnBurst_Instantaneous001}
  ]
  particleSpawnScript [
   &InitializeParticle(
    Color Mode=::NewEnumerator0
    Lifetime=$System.CurrentLoopDuration
    Mesh Scale Mode=::NewEnumerator3
    Mesh Uniform Scale=1.5
    UsePositionOffset=false
    Write Material Random=true
   )
   &InitialMeshOrientation(Mesh Axis To Orient=<0 0 1>, Mesh Orientation Mode=::NewEnumerator2, Use Rotation Vector=true)
  ]
  particleUpdateScript [
   &ParticleState()
   &CameraOffset(Camera Offset Amount=38.6542)
   &ScaleColor(
    Scale Alpha=&FloatFromCurve(
     FloatCurve=DataInterface() {propertyValues {Curve {keys [
         {interpMode=RCIM_Cubic, tangentMode=RCTM_User, time=0, value=0}
         {interpMode=RCIM_Cubic, tangentMode=RCTM_User, time=0.2, value=1}
         {interpMode=RCIM_Cubic, tangentMode=RCTM_User, time=1, value=0}
        ]}}}
     Scale Curve=$System.InterpolateOverTime.InterpolateOverTime_Float.Moving Average
    )
    Scale RGB=&VectorFromFloat(Value=8)
    ScaleA=true
    ScaleRGB=true
   )
   &DynamicMaterialParameters(Module.Index0Param0Write=true, Module.Index0Param1Write=true, Module.Index0Param2Write=true, Module.Index0Param3Write=true)
  ]
  rendererClasses [&NiagaraMeshRendererProperties]
  renderers [{rendererClass=&NiagaraMeshRendererProperties, rendererIndex=0}]
 }
 EnergyBeam {
  emitterUpdateScript [
   &EmitterState()
   &SpawnBurst_Instantaneous(Spawn Count=2, Use Spawn Probability=false, UseLoopCountLimit=false)
  ]
  particleSpawnScript [&InitializeParticle(
    Color=<rgba 1 0 0.0833 0.9>
    Lifetime=&Multiply_Float(A=$User.Duration, B=0.6)
    Mesh Scale Mode=::NewEnumerator3
    Mesh Uniform Scale=139.1489
    Position Mode=::NewEnumerator0
    Ribbon Width=86.2598
    Ribbon Width Mode=::NewEnumerator1
    Write Material Random=true
   )]
  particleUpdateScript [
   &ParticleState()
   &DynamicMaterialParameters(
    Index 2 Param 1=1
    Module.Index0Float0Write=false
    Module.Index0Float1Write=true
    Module.Index0Float2Write=true
    Module.Index0Float3Write=true
    Module.Index2Float0Write=true
    Module.Index2Float1Write=true
    Module.Index2Float2Write=true
    Module.Index2Float3Write=true
    Write Parameter Index 2=::NewEnumerator1
   )
   /SetVariables(Particles.Position=&LerpPosition(
     A=/Niagara/DynamicInputs/TypeConversions/MakeCustomPositionFromBool(
      Bool:NiagaraBool=$User.IsMinor
      True Position:NiagaraPosition=$System.SampleSkeletalMesh002.SampledPosition
      False Position:NiagaraPosition=$System.SampleSkeletalMesh.SampledPosition
     )
     B=$System.SampleSkeletalMesh001.SampledPosition
     Alpha=&ReturnNormalizedExecIndex()
    )) {isSetParameters=true}
   &ScaleColor(
    Scale Alpha=&FloatFromCurve(
     FloatCurve=DataInterface() {propertyValues {Curve {
        defaultValue=1
        keys [
         {arriveTangent=7.3658, interpMode=RCIM_Cubic, leaveTangent=7.3658, tangentMode=RCTM_User, time=0, value=0}
         {interpMode=RCIM_Cubic, tangentMode=RCTM_User, time=0.5, value=1}
         {tangentMode=RCTM_User, time=1, value=0}
        ]
       }}}
     Scale Curve=$System.InterpolateOverTime.InterpolateOverTime_Float.Moving Average
    )
    ScaleA=true
    ScaleRGB=true
   )
   &FN_Sprite_Attributes(
    Color=&MultiplyLinearColor(
     Linear Color=$Particles.Color
     Scale Factor=&MakeVector4(
      X=5
      Y=5
      Z=5
      W=&FloatFromCurve(
       FloatCurve=DataInterface() {propertyValues {Curve {
          defaultValue=1
          keys [
           {arriveTangent=7.3658, interpMode=RCIM_Cubic, leaveTangent=7.3658, tangentMode=RCTM_User, time=0, value=0}
           {interpMode=RCIM_Cubic, tangentMode=RCTM_User, time=0.5, value=1}
           {tangentMode=RCTM_User, time=1, value=0}
          ]
         }}}
       Scale Curve=3
      )
     )
    )
    Create Color Attribute=true
    Write to Module Namespace=true
   )
   /SetVariables(Particles.B2=3) {isSetParameters=true}
  ]
  rendererClasses [&NiagaraRibbonRendererProperties]
  renderers [
   {rendererClass=&NiagaraRibbonRendererProperties}
   {rendererClass=&NiagaraRibbonRendererProperties, rendererIndex=1}
  ]
 }
 TelegraphDecal {
  emitterUpdateScript [
   &EmitterState(
    Inactive Response=::NewEnumerator1
    Life Cycle Mode=::NewEnumerator1
    Loop Behavior=::NewEnumerator1
    Loop Duration=$User.Duration
    UseLoopDelay=false
   )
   &SpawnBurst_Instantaneous(Spawn Count=1, Use Spawn Probability=false)
   /SetVariables(Emitter.OverallOpacity:NiagaraFloat=&Lerp_Float(A=0, B=0.4, Alpha=$System.InterpolateOverTime.InterpolateOverTime_Float.Moving Average)) {isSetParameters=true}
  ]
  particleSpawnScript [
   &InitializeParticle(Lifetime=3, Sprite Size Mode=::NewEnumerator3, Uniform Sprite Size=600, UsePositionOffset=false, Write Material Random=true)
   /SetVariables(
    Particles.DecalFade=1
    Particles.DecalOrientation:Quat4f=&OrientMeshToVector(Look At Direction=<0 0 1>, Reference Axis Vector=<1 0 0>)
    Particles.DecalSize:Vector3f=<300 400 400>
   ) {isSetParameters=true}
  ]
  particleUpdateScript [
   &ParticleState(Kill Particles When Lifetime Has Elapsed=false, Let Infinitely Lived Particles Die When Emitter Deactivates=true)
   /SetVariables(Particles.DecalFade:NiagaraFloat=&Multiply_Float(
     A=$System.InterpolateOverTime.InterpolateOverTime_Float.Moving Average
     B=&Lerp_Float(
      A=0
      B=$System.Timeline.Float
      Alpha=&FloatFromCurve(FloatCurve=DataInterface() {propertyValues {Curve {keys [
           {arriveTangent=12.3499, interpMode=RCIM_Cubic, leaveTangent=12.3499, tangentMode=RCTM_User, time=0, value=0}
           {arriveTangent=0.2129, interpMode=RCIM_Cubic, leaveTangent=0.2129, tangentMode=RCTM_User, time=0.2, value=1}
           {arriveTangent=-0.0037, interpMode=RCIM_Cubic, leaveTangent=-0.0037, tangentMode=RCTM_User, time=1, value=1}
          ]}}})
     )
    )) {isSetParameters=true}
  ]
  rendererClasses [&NiagaraDecalRendererProperties]
  renderers [{rendererClass=&NiagaraDecalRendererProperties, rendererIndex=0}]
 }
}
summary {
 systemName=NS_Spire_PlayerWarning
 emitters [
  {emitterName=EdgeGlow, rendererClasses [&NiagaraMeshRendererProperties]}
  {emitterName=EnergyBeam, rendererClasses [&NiagaraRibbonRendererProperties]}
  {emitterName=TelegraphDecal, rendererClasses [&NiagaraDecalRendererProperties]}
 ]
 userVariables [
  User.bool {type=NiagaraBool}
  User.Duration {type=NiagaraFloat, defaultValue=2, description=""}
  User.InterpolateRateOfChange {type=NiagaraFloat, defaultValue=1, description=""}
  User.IsAboutToAttack {type=NiagaraBool, defaultValue=true}
  User.IsMinor {type=NiagaraBool}
  User.Position_Spire {type=NiagaraPosition, description="", defaultValue {}}
  User.SKMesh {
   type=NiagaraDataInterfaceSkeletalMesh
   defaultValue {
    struct=NiagaraExt_VariableValue_DataInterface
    value {dataInterface=&self:NiagaraDataInterfaceSkeletalMesh_0}
   }
  }
  User.SKMesh_Minor {
   type=NiagaraDataInterfaceSkeletalMesh
   defaultValue {
    struct=NiagaraExt_VariableValue_DataInterface
    value {dataInterface=&self:NiagaraDataInterfaceSkeletalMesh_1}
   }
  }
 ]
}
